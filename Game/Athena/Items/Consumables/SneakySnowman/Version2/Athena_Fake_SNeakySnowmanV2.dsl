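; Interactable fake-snowman actor for the Sneaky Snowman V2 consumable, offering an instant "Hide" interaction.
; BlueprintCanInteract admits only non-teammates of the wearer holding no object, while the wearer is crouched, attached and not clipping, subject to gameplay-tag checks.
; Line of sight requires both low and high traces to hit nothing, so it works inside collisionless props.
; Interacting forwards the pawn to the owning player snowman's OnInteractwithFake.
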(event EventBeginPlay)

(event Collision|EventActorBeginOverlap
  (params (OtherActor :type Actor%20Object%20Reference)))

(event EventTick
  (params (DeltaSeconds :type Float%20%28single-precision%29)))

(event Interaction|EventBlueprintOnInteract
  (params (InteractingPawn :type Fort%20Pawn%20Object%20Reference) (InteractionBeingAttempted :type EInteractionBeingAttempted%20Enum :default "FirstInteraction"))
  (Class|AthenaPlayerSneakySnowmanV2|OnInteractwithFake (Variables|Default|GetMyOwner) InteractingPawn))

(fn BlueprintCanInteract
  (params (InteractingPawn :type Fort%20Pawn%20Object%20Reference) (InteractionBeingAttempted :type EInteractionBeingAttempted%20Enum :default "FirstInteraction") (InteractionType :type TInteractionType%20Enum :default "IT_NoInteraction"))
  (Utilities|IsValid (Variables|Default|GetMyOwner)
    (:Is%20Valid
      (bind _output_get (LocalVariables|SetPlayerPawn (Variables|/Game/Athena/Items/Consumables/SneakySnowman/Version2/Athena_Player_SneakySnowmanV2.Athena_Player_SneakySnowmanV2_C|GetPlayerPawn (Variables|Default|GetMyOwner))))
      (if (Utilities|NotEqual(Object) _output_get InteractingPawn)
        (Utilities|IsValid (LocalVariables|GetPlayerPawn)
          (:Is%20Valid
            (Utilities|IsValid InteractingPawn
              (:Is%20Valid
                (bind _asfort_player_pawn (Utilities|Casting|CastToFortPlayerPawn InteractingPawn)
                  (:then
                    (bind _returnvalue (Fortnite|HeldObjects|GetHeldObjectAttachedtoPlayer _asfort_player_pawn))
                    (Utilities|IsValid _returnvalue
                      (:Is%20Valid
                        _)
                      (:Is%20Not%20Valid
                        (if (not (Team|OnSameTeam InteractingPawn (LocalVariables|GetPlayerPawn)))
                          ;[1 1 1] Trace for Line Of Sight To Player. If the trace hits NOTHING, then you have line of sight. This is so the line of sight check works inside props that have no player collision. (Big Bush, Corn field)
                          (bind (_outhit _returnvalue_2) (Collision|LineTraceByChannel (+ (Transformation|GetActorLocation InteractingPawn) "0,0.000000,35.000000") (+ (Transformation|GetActorLocation) "0,0.000000,35.000000") "TraceTypeQuery2" true (Utilities|Array|MakeArray (Variables|Default|GetMyOwner) self (Variables|/Game/Athena/Items/Consumables/SneakySnowman/Version2/Athena_Player_SneakySnowmanV2.Athena_Player_SneakySnowmanV2_C|GetPlayerPawn (Variables|Default|GetMyOwner))) "None" false))
                          (bind (_outhit_2 _returnvalue_3) (Collision|LineTraceByChannel (+ (Transformation|GetActorLocation InteractingPawn) "0,0.000000,200.000000") (+ (Transformation|GetActorLocation) "0,0.000000,200.000000") "TraceTypeQuery2" true (Utilities|Array|MakeArray (Variables|Default|GetMyOwner) self (Variables|/Game/Athena/Items/Consumables/SneakySnowman/Version2/Athena_Player_SneakySnowmanV2.Athena_Player_SneakySnowmanV2_C|GetPlayerPawn (Variables|Default|GetMyOwner))) "None" false))
                          (return (Math|Boolean|ANDBoolean (and (not _returnvalue_3) (not _returnvalue_2)) (AI|Components|NavMovement|IsCrouching (Variables|/Script/Engine.Character|GetCharacterMovement (LocalVariables|GetPlayerPawn))) (and (or (not (GameplayTags|HasAnyMatchingGameplayTags (GameplayTags|MakeGameplayTagContainerfromArray (Utilities|Array|MakeArray "(TagName=\"Gameplay.Status.HidingInProp\")" "(TagName=\"Gameplay.Status.HidingProp.GettingInProp\")")))) (GameplayTags|HasMatchingGameplayTag "(TagName=\"Athena.Item.Snowman\")")) (not (GameplayTags|HasAnyMatchingGameplayTags (GameplayTags|MakeGameplayTagContainerfromArray (Utilities|Array|MakeArray "(TagName=\"Gameplay.Action.Player.AthenaInVehicle\")" "(TagName=\"MovementMode.Falling\")" "(TagName=\"Gameplay.Status.HidingProp.BlockInteract\")" "(TagName=\"Gameplay.Action.Player.CarryingPlayer\")"))))) (Variables|/Game/Athena/Items/Consumables/SneakySnowman/Version2/Athena_Player_SneakySnowmanV2.Athena_Player_SneakySnowmanV2_C|GetAttached (Variables|Default|GetMyOwner)) (not (Variables|Default|GetClipping)))))))
                    (return false))
                  (:CastFailed
                    _)))
              (:Is%20Not%20Valid
                _)))
          (:Is%20Not%20Valid
            _))))
    (:Is%20Not%20Valid
      _)))

(fn BlueprintGetInteractionString
  (params (InteractingPawn :type Fort%20Pawn%20Object%20Reference) (InteractionBeingAttempted :type EInteractionBeingAttempted%20Enum :default "FirstInteraction"))
  (return "NSLOCTEXT(\"[3D506EEB44BD1DCB5B2CD0AFF8F9B149]\", \"F862581347315FBFCA86DC8D9F02923B\", \"Hide\")"))

(fn BlueprintGetInteractionTime
  (params (InteractingPawn :type Fort%20Pawn%20Object%20Reference) (InteractionBeingAttempted :type EInteractionBeingAttempted%20Enum :default "FirstInteraction"))
  (return 0.0 true))

(fn ConstructionScript)
