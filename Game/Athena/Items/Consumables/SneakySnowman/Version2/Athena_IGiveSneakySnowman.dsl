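; Blueprint interface for giving the Sneaky Snowman consumable, version 2, by its name and path.
; It declares queries for a material integer and a new snowman transform, plus calls to move a player actor and spawn particles.
; Notification functions cover the snowman being given or taken and a jump-in starting.
; All functions are empty or return default values, so the behavior is not defined here.
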(fn GetMaterialInt
  (return 0))

(fn GetNewSnowmanTransform
  (return 0))

(fn MovePlayer
  (params (Player :type Actor%20Object%20Reference)))

(fn OnGaveSneakySnowman)

(fn OnStartedJumpingIn)

(fn OnTookSneakySnowman)

(fn SpawnParticles)
