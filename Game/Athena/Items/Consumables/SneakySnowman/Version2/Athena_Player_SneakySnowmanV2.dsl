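; Sneaky Snowman V2 disguise actor that attaches to a player's pelvis and hides them from players and AI.
; A 0.1s timer and movement updates drive swapping the worn mesh for a spawned fake actor while the player crouches still, tilted to the ground via timelines.
; Jumping out spawns a prop in its place.
; Damage, DBNO, death, interaction by others, weapon fire, vehicles or excessive drift from the fake can destroy it.
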(event Custom|StateChangedDelegate_Event_0
  (params (Minigame :type Fort%20Minigame%20Object%20Reference) (MinigameState :type EFortMinigameState%20Enum :default "PreGame"))
  ;[1 1 1] Makes sure snowman is destroyed when returning to the hub or starting a minigame in Creative
  (switch Utilities|FlowControl|Switch|SwitchonEFortMinigameState MinigameState
    (:Transitioning
      (CallFunction|Destroy))))

(event Custom|OnPlayerGotSnowman
  (bind (_returnvalue_roll _returnvalue_pitch _returnvalue_yaw) (Math|Rotator|BreakRotator (Pawn|GetControlRotation (Variables|StoredActors|GetPlayerPawn) :!class /Script/Engine.Pawn)))
  (bind (_adjust_startzrot_roll _adjust_startzrot_pitch _adjust_startzrot_yaw) (Math|Rotator|BreakRotator (Variables|GetAdjust_StartZRot_Roll)))
  (bind (_rotationoffsetonattach_x _rotationoffsetonattach_y _rotationoffsetonattach_z) (Math|Vector|BreakVector (Variables|GetRotationOffsetOnAttach_X)))
  (bind (_adjust_startzrot_roll_2 _adjust_startzrot_pitch_2) (Math|Rotator|BreakRotator (Variables|GetAdjust_StartZRot_Roll)))
  (bind (_returnvalue_roll_2 _returnvalue_pitch_2 _returnvalue_yaw_2) (Math|Rotator|BreakRotator (Transformation|GetActorRotation)))
  (Utilities|FlowControl|Sequence
    (:then_0
      ;[1 1 1] Makes sure snowman is destroyed when returning to the hub or starting a minigame in Creative
      (Networking|SwitchHasAuthority
        (:Authority
          (bind (_asynctaskproxy _payload) (GameplayEventRouter|StartListeningToEventBackwardCompatible "(TagName=\"EventMessage.Component.Playspace.PlayerExitedSpatialActor.LegacyMinigame\")" (GameplayEventRouter|GetGlobalEventRouter) "/Script/FortniteGame.PlayerExitSpatialActorContextLegacyMinigame" 0 "Down")
            (:then
              (Variables|Default|SetExitedEventHandle _asynctaskproxy)
              (bind _playerpawn (Variables|StoredActors|GetPlayerPawn)
                (:then
                  (bind _playerstate (Variables|/Script/Engine.Pawn|GetPlayerState _playerpawn)
                    (:then
                      (Utilities|IsValid (Minigame|GetMinigame _playerstate :!class /Script/FortniteGame.MinigameBlueprintLibrary)
                        (:Is%20Valid
                          (Minigame|AssignStateChangedDelegate (Minigame|GetMinigame _playerstate :!class /Script/FortniteGame.MinigameBlueprintLibrary) (delegate Custom|StateChangedDelegate_Event_0)))))
                    (:else)))
                (:else)))
            (:OnEventReceived
              (bind _playerpawn_2 (Variables|StoredActors|GetPlayerPawn)
                (:then
                  (bind _playerstate_2 (Variables|/Script/Engine.Pawn|GetPlayerState _playerpawn_2)
                    (:then
                      (if (Utilities|Equal(Object) _playerstate_2 (.ExitingPlayerState (Utilities|Struct|BreakPlayerExitSpatialActorContextLegacyMinigame _payload)))
                        (CallFunction|Destroy)))
                    (:else)))
                (:else)))))))
    (:then_1
      ;[1 1 1] Attaching to Player
      (Rendering|SetActorHiddenInGame false)
      (Variables|CachedValues|SetAdjust_StartZRot (Transformation|GetActorRotation))
      (Variables|CachedValues|SetAdjust_GoalRot (Transformation|GetSocketRotation (Variables|/Script/Engine.Character|GetMesh (Variables|StoredActors|GetPlayerPawn)) "Pelvis"))
      (CallFunction|Wobble 1.0)
      (Transformation|AttachActorToComponent (Variables|/Script/Engine.Character|GetMesh (Variables|StoredActors|GetPlayerPawn)) "Pelvis" "SnapToTarget" "KeepWorld" "KeepWorld" false)
      (Variables|CachedValues|SetAdjust_StartLocation (Variables|/Script/Engine.SceneComponent|GetRelativeLocation (Variables|Default|GetScene)))
      (bind (_direction _progress) (Timeline|AdjustSnowman Play)
        (:Update
          (if (Variables|DesignVariables|GetRotateOnEnter)
            (Transformation|SetActorRotation (Math|Rotator|Lerp(Rotator) (Math|Rotator|MakeRotator _adjust_startzrot_roll _adjust_startzrot_pitch _adjust_startzrot_yaw) (Math|Rotator|MakeRotator (+ _adjust_startzrot_roll_2 _rotationoffsetonattach_x) (+ _adjust_startzrot_pitch_2 _rotationoffsetonattach_y) (+ _returnvalue_yaw _rotationoffsetonattach_z)) _progress true) true)
            (else
              ;; ResetXYRotation
              (Transformation|SetActorRotation 0 (Math|Rotator|MakeRotator 0.0 0.0 _returnvalue_yaw_2) true))))
        (:Finished
          (if (not (Variables|StateTrackers|GetCurrentlyReplacedByFake))
            (CallFunction|SetSnowmanHidden false))
          (Variables|StateTrackers|SetAttached true)
          (Fortnite|SetCanbeDamaged self true)
          (Utilities|ApplyEffect "/Game/Athena/Items/Consumables/SneakySnowman/Version2/GE_BlockExplosions.GE_BlockExplosions_C" (Variables|StoredActors|GetPlayerPawn))
          (Networking|ForceNetUpdate))))))

(event Custom|DBNO
  (params (FortPawn :type Fort%20Pawn%20Object%20Reference) (bInIsDBNO :type Boolean))
  ;[1 1 1] Destroy Snowman on DBNO or Elimination
  (if bInIsDBNO
    (Ability|SendGameplayEventtoActor (Variables|Actor|GetInstigator) "(TagName=\"Abilities.Generic.ActiveAbility.Activate\")")
    (Actor|DestroyActor)))

(event Custom|JustAttached
  ;[1 1 1] Attaching to Player
  (if (not (Variables|StateTrackers|GetCurrentlyReplacedByFake))
    (CallFunction|SetSnowmanHidden false))
  (Variables|StateTrackers|SetAttached true)
  (Fortnite|SetCanbeDamaged self true)
  (Utilities|ApplyEffect "/Game/Athena/Items/Consumables/SneakySnowman/Version2/GE_BlockExplosions.GE_BlockExplosions_C" (Variables|StoredActors|GetPlayerPawn))
  (Networking|ForceNetUpdate))

(event Custom|PlayerDied
  (params (DamagedActor :type Actor%20Object%20Reference) (Damage :type Float%20%28single-precision%29) (InstigatedBy :type Controller%20Object%20Reference) (DamageCauser :type Actor%20Object%20Reference) (HitLocation :type Vector) (FHitComponent :type Primitive%20Component%20Object%20Reference) (BoneName :type Name) (Momentum :type Vector))
  ;[1 1 1] Destroy Snowman on DBNO or Elimination
  (Ability|SendGameplayEventtoActor (Variables|Actor|GetInstigator) "(TagName=\"Abilities.Generic.ActiveAbility.Activate\")")
  (Actor|DestroyActor))

(event Custom|Destroyed
  (params (DestroyedActor :type Actor%20Object%20Reference))
  ;[1 1 1] Clean up on destroyed
  (bind _playerpawn (Variables|StoredActors|GetPlayerPawn)
    (:then
      (Weapon|SetHolsterWeapon _playerpawn 0 false false "None")
      (Utilities|IsValid (Ability|GetAbilitySystemComponent _playerpawn)
        (:Is%20Valid
          (GameplayEffects|RemoveActiveEffectswithGrantedTags (Ability|GetAbilitySystemComponent _playerpawn) "(GameplayTags=((TagName=\"Athena.AI.Hidden\"),(TagName=\"Gameplay.Action.Player.BlockExplosionDamage\"),(TagName=\"Gameplay.Action.Player.PreventADS\"),(TagName=\"Granted.Athena.EnvItem.Sentry.Turret.Hidden\")))"))
        (:Is%20Not%20Valid
          _))
      (CallFunction|ToggleFakeSnowmanMulticast false)
      (bind _instigator (Variables|Actor|GetInstigator)
        (:then
          (Ability|SendGameplayEventtoActor _instigator "(TagName=\"Abilities.Generic.ActiveAbility.Activate\")"))
        (:else
          _))
      (bind _fakeactor (Variables|StoredActors|GetFakeActor)
        (:then
          (Actor|DestroyActor _fakeactor))
        (:else
          _))
      (if (Variables|StateTrackers|GetShouldSpawnDeathFX)
        (Effects|Components|ParticleSystem|SpawnFXAtLocation (Variables|Constants|FX|GetP_Death_Effects) (Transformation|GetActorLocation))
        (Audio|PlaySoundatLocation (Variables|Sound|GetDeath_SoundFX) (Transformation|GetActorLocation))))
    (:else
      _)))

(event Custom|SetPrevLocation
  ;[1 1 1] Leave snowman behind on enter vehicle or ride wildlife
  (Variables|CachedValues|SetPreviousLocation (Transformation|GetActorLocation)))

(event EventBeginPlay
  (bind (_returnvalue_location _returnvalue_rotation) (Math|Transform|BreakTransform (Transformation|GetActorTransform)))
  (bind (_hasmovementbase _movementbase) (CallFunction|GetPlayerMovementBase))
  (bind _returnvalue (Game|GetInstigator))
  (bind (_returnvalue_location_2 _returnvalue_rotation_2 _returnvalue_scale) (Math|Transform|BreakTransform (Transformation|GetActorTransform)))
  (bind (_relativelocation_x _relativelocation_y _relativelocation_z) (Math|Vector|BreakVector (Variables|/Script/Engine.SceneComponent|GetRelativeLocation_X (Variables|Default|GetSM_SneakySnowman))))
  (bind _playerpawn (Variables|StoredActors|GetPlayerPawn))
  (bind _playerpawn_2 (Variables|StoredActors|GetPlayerPawn))
  ;[1 1 1] Begin Play
  (Utilities|FlowControl|Sequence
    (:then_0
      ;[1 1 1] General Setup
      (Collision|SetCollisionEnabled (Variables|Default|GetCylinder) "NoCollision")
      (bind _asfort_player_pawn (Utilities|Casting|CastToFortPlayerPawn _returnvalue))
      (Variables|StoredActors|SetPlayerPawn _asfort_player_pawn)
      (Utilities|Time|SetTimerbyEvent (EventDispatchers|CreateEvent :!event Custom|FrequentChecksTimer) 0.1 true)
      (if (Utilities|IsValid _returnvalue :!class /Script/Engine.KismetSystemLibrary)
        (CallFunction|SetTeam (Team|GetActorTeam _returnvalue))
        (else
          _))
      (Variables|CachedValues|SetWobble_OriginalZ _relativelocation_z)
      (if (not (Variables|StateTrackers|GetCurrentlyReplacedByFake))
        (CallFunction|SetSnowmanHidden (Variables|Constants|GetStartHidden) false))
      (Fortnite|SetCanbeDamaged self) ; this gets set to true once this actor is attached to player
    )
    (:then_1
      ;[1 1 1] All Bind Events
      (Utilities|FlowControl|Sequence
        (:then_0
          ;[1 1 1] Destroy Snowman on DBNO or Elimination
          (if (Health|IsDBNO (Variables|StoredActors|GetPlayerPawn))
            (Actor|DestroyActor)
            (else
              (bind _playerpawn_3 (Variables|StoredActors|GetPlayerPawn)
                (:then
                  (EventDispatchers|BindEventtoOnDied _playerpawn_3 (delegate Custom|PlayerDied))
                  (bind _playerpawn_4 (Variables|StoredActors|GetPlayerPawn)
                    (:then
                      (EventDispatchers|BindEventtoOnDBNOStateChanged _playerpawn_4 (delegate Custom|DBNO)))
                    (:else)))
                (:else)))))
        (:then_1
          (Game|BindEventtoOnDestroyed (delegate Custom|Destroyed)))
        (:then_2
          ;[1 1 1] Leave snowman behind on enter vehicle or ride wildlife
          (Networking|SwitchHasAuthority
            (:Authority
              (Utilities|Time|SetTimerbyEvent (delegate Custom|SetPrevLocation) 0.2 true)
              (Utilities|IsValid _playerpawn
                (:Is%20Valid
                  (Ability|Tasks|WaitGameplayTagAddToActor _playerpawn "(TagName=\"Gameplay.Action.Player.AthenaInVehicle\")" true
                    (:then
                      (Ability|Tasks|WaitGameplayTagAddToActor _playerpawn "(TagName=\"Gameplay.Action.Player.Grinding\")" true
                        (:then
                          (Ability|Tasks|WaitGameplayTagAddToActor _playerpawn "(TagName=\"Granted.Riding.Player.IsRiding\")" true
                            (:Added
                              (CallFunction|OnInvalidInteractions true))))
                        (:Added
                          _)))
                    (:Added
                      _)) ; Wait for the vehicle tag to be applied instead of binding to events.\nThis will resolve some race conditions like FORT-432518
                  (CallFunction|OnInvalidInteractions false))))))
        (:then_3
          ;[1 1 1] Logic for maintaining rotation when player is crouched and not moving
          (Networking|SwitchHasAuthority
            (:Authority
              (Character|BindEventtoOnCharacterMovementUpdated (Variables|StoredActors|GetPlayerPawn) (delegate Custom|MovementUpdated)))))))
    (:then_2
      ;[1 1 1] Drop and destroy any held objects
      (bind _returnvalue_2 (Fortnite|HeldObjects|GetHeldObjectComponentAttachedtoPlayer (Variables|StoredActors|GetPlayerPawn)))
      (Utilities|IsValid _returnvalue_2
        (:Is%20Valid
          (HeldObjects|Behavior|DropHeldObject _returnvalue_2)
          (Actor|DestroyActor (Components|GetOwner _returnvalue_2)))))
    (:then_3
      ;[1 1 1] Spawn Interact Sphere (to handle unhiding)
      (Networking|SwitchHasAuthority
        (:Authority
          (bind _returnvalue_3 (Game|SpawnActorfromClass "/Game/Athena/Items/Consumables/SneakySnowman/Version2/Athena_SneakySnowmanV2_InteractSphere.Athena_SneakySnowmanV2_InteractSphere_C" (Math|Transform|MakeTransform _returnvalue_location_2 _returnvalue_rotation_2 _returnvalue_scale) "AlwaysSpawn" "OverrideRootScale" 0 0 (Game|GetInstigator)))
          (Variables|StoredActors|SetInteractSphereRef _returnvalue_3)
          (Transformation|AttachActorToActor _returnvalue_3 (Variables|StoredActors|GetPlayerPawn) "None" "SnapToTarget" "KeepWorld" "KeepWorld" false)
          (Transformation|SetActorRelativeLocation _returnvalue_3 "0.000000, 0,50.000000" false true)
          (Class|AthenaSneakySnowmanV2InteractSphere|Init _returnvalue_3 self))))
    (:then_4
      ;[1 1 1] Spawn fake actor (replaces this actor while crouched and not moving)
      (bind _returnvalue_4 (Building|SpawnBuildingGameplayActor (Variables|Constants|GetFakeActorClass) (Math|Transform|MakeTransform _returnvalue_location _returnvalue_rotation) self false false))
      (bind _asathena_fake_sneaky_snowman_v2 (Utilities|Casting|CastToAthena_Fake_SneakySnowmanV2 _returnvalue_4)
        (:then
          (Variables|/Game/Athena/Items/Consumables/SneakySnowman/Version2/Athena_Fake_SNeakySnowmanV2.Athena_Fake_SneakySnowmanV2_C|SetMyOwner self _asathena_fake_sneaky_snowman_v2))
        (:CastFailed
          _))
      (CallFunction|SetFakeActor _returnvalue_4)
      (Utilities|IsValid _returnvalue_4
        (:Is%20Valid
          (if _hasmovementbase
            (Transformation|AttachActorToActor _returnvalue_4 _movementbase "None" "KeepWorld" "KeepWorld" "KeepWorld" false)))))
    (:then_5
      ;[1 1 1] EventBindings
      (Ability|Async|WaitGameplayEventToActor _playerpawn_2 "(TagName=\"Event.Primary.Pressed\")"
        (:then
          (if (or (ScalableFloat|GetValueasBool (Variables|DesignVariables|GetCanCrouchOnMovingPlatform)) (CallFunction|AttachestoMovingPlatform))
            (EventDispatchers|AssignOnMovementBaseChanged _playerpawn_2 (delegate Custom|HandleMovementBaseChanged))))
        (:EventReceived
          (if (Variables|Default|GetDestroyWhenSwappingOrShooting)
            (CallFunction|ServerClickedFire)))))))

(event Custom|OnInvalidInteractions
  (params (SpawnActor :type Boolean) (PlayDeathVFX :type Boolean))
  ;[1 1 1] Leave snowman behind on enter vehicle or ride wildlife
  (Ability|SendGameplayEventtoActor (Variables|Actor|GetInstigator) "(TagName=\"Abilities.Generic.ActiveAbility.Activate\")")
  (CallFunction|ToggleFakeSnowmanMulticast false)
  (Variables|StateTrackers|SetShouldSpawnDeathFX PlayDeathVFX)
  (Actor|DestroyActor)
  (Networking|SwitchHasAuthority
    (:Authority
      (if SpawnActor
        (bind _returnvalue (Building|SpawnBuildingActor (Variables|Constants|GetPropToSpawnOnExit) (Math|Transform|MakeTransform (Variables|CachedValues|GetPreviousLocation)) 0 (Variables|StoredActors|GetPlayerPawn)))
        (Default|CallOnSpawnProp _returnvalue)))))

(event Custom|Server_ClickedFire
  (bind _playerpawn (Variables|StoredActors|GetPlayerPawn)
    (:then
      (Utilities|IsValid (Weapon|GetCurrentWeapon _playerpawn)
        (:Is%20Valid
          (if (and (CallFunction|ShouldDestroywhenUsingWeapon (Weapon|GetCurrentWeapon _playerpawn)) (not (GameplayTags|HasMatchingGameplayTag _playerpawn "(TagName=\"Abilities.SneakySnowman.JumpingIn\")")))
            (CallFunction|OnInvalidInteractions false true)))))
    (:else)))

(event Custom|SetFakeActor
  (params (FakeActor :type Building%20Gameplay%20Actor%20Object%20Reference))
  ;[1 1 1] Spawn fake actor (replaces this actor while crouched and not moving)
  (bind _output_get (Variables|StoredActors|SetFakeActor FakeActor))
  (Rendering|SetActorHiddenInGame _output_get true))

(event Custom|SpawnEffects
  (Effects|Components|ParticleSystem|SpawnFXAtLocation (Variables|Constants|FX|GetP_Explosion) (Transformation|GetActorLocation)))

(event Custom|ShowFakeActor
  ;[1 1 1] Logic for maintaining rotation when player is crouched and not moving
  (Rendering|SetActorHiddenInGame (Variables|StoredActors|GetFakeActor) false))

(event Custom|HandleMovementBaseChanged
  (params (NewBaseComponent :type Primitive%20Component%20Object%20Reference) (BoneName :type Name))
  (bind _returnvalue (Components|GetOwner NewBaseComponent))
  (bind _returnvalue_2 (Class|FortBaseMovingPlatformComponent|FindMovingPlatformComponentfromActor _returnvalue true))
  (bind _returnvalue_3 (Components|GetOwner _returnvalue_2))
  (bind _fakeactor (Variables|StoredActors|GetFakeActor))
  ;[1 1 1] EventBindings
  (Utilities|IsValid _returnvalue
    (:Is%20Valid
      (Utilities|IsValid _returnvalue_2
        (:Is%20Valid
          (Utilities|IsValid _returnvalue_3
            (:Is%20Valid
              (Transformation|AttachActorToActor _fakeactor _returnvalue_3 "None" "KeepWorld" "KeepWorld" "KeepWorld" false))
            (:Is%20Not%20Valid
              _)))
        (:Is%20Not%20Valid
          _)))
    (:Is%20Not%20Valid
      _))
  (Transformation|DetachFromActor _fakeactor "KeepWorld" "KeepWorld" "KeepWorld"))

(event Custom|MovementUpdated
  (params (DeltaSeconds :type Float%20%28single-precision%29) (OldLocation :type Vector) (OldVelocity :type Vector))
  (bind _playerpawn (Variables|StoredActors|GetPlayerPawn))
  (bind _returnvalue (GameplayTags|HasMatchingGameplayTag _playerpawn "(TagName=\"Gameplay.Action.Player.SurfaceSwimming\")"))
  ;[1 1 1] Logic for maintaining rotation when player is crouched and not moving
  (Utilities|FlowControl|Sequence
    (:then_0
      (if (Variables|StateTrackers|GetCurrentlyReplacedByFake)
        (CallFunction|ShowFakeActor) ; Fixes a rare bug where the fake\nactor turns invisible when it shouldn't
      ))
    (:then_1
      (CallFunction|UpdateSnowmanState)
      ;[1 1 1] Block ADS while swimming
      (Networking|SwitchHasAuthority
        (:Authority
          (if (not (Math|Boolean|Equal(Boolean) (Variables|StateTrackers|GetSwimming) _returnvalue))
            (bind _output_get (Variables|StateTrackers|SetSwimming _returnvalue))
            (if _output_get
              (Utilities|ApplyEffect "/Game/Athena/Items/Consumables/SneakySnowman/CoolCarpet/GE_BlockADS_Infinite.GE_BlockADS_Infinite_C" (Variables|StoredActors|GetPlayerPawn))
              (else
                (GameplayEffects|RemoveActiveEffectswithGrantedTags (Ability|GetAbilitySystemComponent (Variables|StoredActors|GetPlayerPawn)) "(GameplayTags=((TagName=\"Gameplay.Action.Player.PreventADS\")))")))))))))

(event Custom|RotateToGround
  (bind (_fake_startrot_roll _fake_startrot_pitch _fake_startrot_yaw) (Math|Rotator|BreakRotator (Variables|GetFake_StartRot_Roll)))
  ;[1 1 1] Logic for maintaining rotation when player is crouched and not moving
  ;;[0.7344 0.7344 0.7344] Interpolate Fake Actor to match angle of the ground
  (bind (_direction _rot) (Timeline|RotateToGroundTimeline PlayFromStart)
    (:Update
      (CallFunction|UpdateActorLocationandRotation (Variables|StoredActors|GetFakeActor) (Utilities|GetFakeGoalPosition) (Math|Rotator|Lerp(Rotator) (Variables|CachedValues|GetFake_StartRot) (Variables|CachedValues|GetFake_GoalRot) _rot true)))
    (:Finished
      (Variables|SetFake_StartRot_Roll (Math|Rotator|MakeRotator 0.0 0.0 _fake_startrot_yaw)))))

(event Custom|StopRotatingFake
  (bind (_fake_startrot_roll _fake_startrot_pitch _fake_startrot_yaw) (Math|Rotator|BreakRotator (Variables|GetFake_StartRot_Roll)))
  ;[1 1 1] Logic for maintaining rotation when player is crouched and not moving
  ;;[0.7344 0.7344 0.7344] Interpolate Fake Actor to match angle of the ground
  (Timeline|RotateToGroundTimeline Stop))

(event Custom|RotateToStraight
  (bind (_fake_startrot_roll _fake_startrot_pitch _fake_startrot_yaw) (Math|Rotator|BreakRotator (Variables|GetFake_StartRot_Roll)))
  ;[1 1 1] Logic for maintaining rotation when player is crouched and not moving
  ;;[0.7344 0.7344 0.7344] Interpolate Fake Actor to match angle of the ground
  (Timeline|RotateToGroundTimeline ReverseFromEnd))

(event Custom|Wobble
  (params (Strength :type Float%20%28double-precision%29 :default 1.0) (MoveUpAndDown :type Boolean))
  (bind _sm_sneakysnowman (Variables|Default|GetSM_SneakySnowman))
  ;[1 1 1] Wobble
  (if MoveUpAndDown
    (Utilities|FlowControl|Sequence
      (:then_0
        (bind (_direction _newtrack_0) (Timeline|UpDown PlayFromStart)
          (:Update
            (Transformation|SetRelativeLocation (Variables|Default|GetSM_SneakySnowman) (Math|Vector|MakeVector 0.0 0.0 (+ (* _newtrack_0 35.0) (Variables|CachedValues|GetWobble_OriginalZ))) false true))))
      (:then_1
        _)))
  (Networking|CanExecuteCosmeticEvents
    (:True
      (Rendering|Material|SetScalarParameterValueonMaterials _sm_sneakysnowman "WobbleStartTime" (Utilities|Time|GetGameTimeinSeconds))
      (Rendering|Material|SetScalarParameterValueonMaterials _sm_sneakysnowman "WobbleStrength" Strength))))

(event Custom|ToggleFakeSnowman
  (params (ToggleFakeOn :type Boolean) (HidePlayerIfTrue :type Boolean) (CrouchedLoc :type Vector) (RotateToGround :type Boolean) (SetPoppingOut :type Boolean) (CanReplaceWithProp :type Boolean))
  (bind (_returnvalue_roll _returnvalue_pitch _returnvalue_yaw) (Math|Rotator|BreakRotator (Transformation|GetActorRotation (Variables|StoredActors|GetFakeActor))))
  (bind (_returnvalue_roll_2 _returnvalue_pitch_2 _returnvalue_yaw_2) (Math|Rotator|BreakRotator (Transformation|GetActorRotation)))
  (bind _playerpawn (Variables|StoredActors|GetPlayerPawn))
  (bind _returnvalue (Ability|GetAbilitySystemComponent _playerpawn))
  (bind (_fake_startrot_roll _fake_startrot_pitch _fake_startrot_yaw) (Math|Rotator|BreakRotator (Variables|GetFake_StartRot_Roll)))
  ;[1 1 1] Logic for maintaining rotation when player is crouched and not moving
  (Variables|CachedValues|SetTFS_HidePlayerIfUsingFake HidePlayerIfTrue)
  (Variables|CachedValues|SetTFS_CrouchedLocation CrouchedLoc)
  (Variables|CachedValues|SetTFS_RotateFakeToGroundAngle RotateToGround)
  (if SetPoppingOut
    (Variables|StateTrackers|SetIsPoppingOut true)
    (else
      _))
  (if ToggleFakeOn
    (if (Variables|StateTrackers|GetIsBeingReplacedWithProp)
      _
      (elif CanReplaceWithProp
        (Variables|StateTrackers|SetCurrentlyReplacedByFake true)
        (Utilities|ApplyEffect "/Game/Athena/Items/Consumables/SneakySnowman/CoolCarpet/GE_HiddenFromAI.GE_HiddenFromAI_C" (Variables|StoredActors|GetPlayerPawn)) ; Hide from AI
        (CallFunction|OnStartedUsingFake) ; override-able funciton
        (Utilities|EnableorDisableFakeActor true)
        (CallFunction|SetSnowmanHidden true)
        (CallFunction|SetPlayerHidden (Variables|CachedValues|GetTFS_HidePlayerIfUsingFake))
        ;;[0.7344 0.7344 0.7344] Interpolate Fake Actor to match angle of the ground
        (Variables|CachedValues|SetFake_StartRot (Transformation|GetActorRotation))
        (bind _returnvalue_2 (CallFunction|GetFakeRotation))
        (Variables|CachedValues|SetFake_GoalRot _returnvalue_2)
        (if (Variables|CachedValues|GetTFS_RotateFakeToGroundAngle)
          (CallFunction|UpdateActorLocationandRotation (Variables|StoredActors|GetFakeActor) (Variables|CachedValues|GetTFS_CrouchedLocation) (Transformation|GetActorRotation))
          (Timeline|RotateToGroundTimeline PlayFromStart)
          (else
            (CallFunction|UpdateActorLocationandRotation 0 (Variables|StoredActors|GetFakeActor) (Utilities|GetFakeGoalPosition) (Math|Rotator|MakeRotator 0.0 0.0 _returnvalue_yaw_2))))))
    (else
      (CallFunction|StopRotatingFake) ; If still rotating fake from when it was toggled on,\ninterrupt it.
      (Variables|StateTrackers|SetCurrentlyReplacedByFake false)
      (Utilities|IsValid _returnvalue
        (:Is%20Valid
          (GameplayEffects|RemoveActiveEffectswithGrantedTags _returnvalue "(GameplayTags=((TagName=\"Athena.AI.Hidden\")))") ; Unhide from AI
        )
        (:Is%20Not%20Valid
          _))
      (CallFunction|OnStoppedUsingFake) ; override-able function
      (Utilities|EnableorDisableFakeActor false)
      (CallFunction|SetSnowmanHidden (Variables|StateTrackers|GetIsBeingReplacedWithProp))
      (CallFunction|SetPlayerHidden false)
      (Input|ResetIgnoreMoveInput (Pawn|GetController (Variables|StoredActors|GetPlayerPawn)))
      (CallFunction|Wobble (Math|Float|SelectFloat 0.35 0.6 (Variables|/Script/Engine.Character|GetbIsCrouched (Pawn|Components|CharacterMovement|GetCharacterOwner (Variables|/Script/Engine.Character|GetCharacterMovement (Variables|StoredActors|GetPlayerPawn))))) true) ; Bounce a bit when player stands up
      (if (not (Variables|DesignVariables|GetSnapRotationToForwardOnStand))
        (Transformation|SetActorRotation 0 (Math|Rotator|MakeRotator 0.0 0.0 _returnvalue_yaw) false) ; Set rotation to match what the fake actor's was (not forward)
      ))))

(event Custom|ToggleFakeSnowman_Multicast
  (params (ToggleFakeOn :type Boolean) (CrouchedLoc :type Vector) (RotateToGround :type Boolean :default true) (SetPoppingOut :type Boolean) (CanReplaceWithProp :type Boolean))
  ;[1 1 1] Logic for maintaining rotation when player is crouched and not moving
  (CallFunction|ToggleFakeSnowman ToggleFakeOn (CallFunction|IsMeshHeightScalePositive) CrouchedLoc RotateToGround SetPoppingOut CanReplaceWithProp))

(event EventOnDamageServer
  (params (Damage :type Float%20%28single-precision%29) (DamageTags :type Gameplay%20Tag%20Container%20Structure%20%28by%20ref%29) (Momentum :type Vector :default "0, 0, 0") (HitInfo :type Hit%20Result%20Structure%20%28by%20ref%29) (InstigatedBy :type Controller%20Object%20Reference) (DamageCauser :type Actor%20Object%20Reference) (EffectContext :type Gameplay%20Effect%20Context%20Handle%20Structure))
  (bind _instigator (Variables|Actor|GetInstigator))
  (bind _returnvalue (Pawn|GetController _instigator))
  (bind _returnvalue_2 (Utilities|Equal(Object) _returnvalue InstigatedBy))
  ;[1 1 1] EVENT -> On Damage - Destroy actor if enough damage has been taken
  (if (Math|Boolean|ORBoolean (not (Team|OnSameTeam InstigatedBy _instigator)) _returnvalue_2 false)
    (bind _output_get (Variables|StateTrackers|SetTotalDamageTaken (+ Damage (Variables|StateTrackers|GetTotalDamageTaken))))
    (if (or (>= _output_get (ScalableFloat|GetValueatLevel (Variables|DesignVariables|GetMaxHealth))) _returnvalue_2)
      (CallFunction|Destroy))))

(event EventOnDeathServer
  (params (Damage :type Float%20%28single-precision%29) (DamageTags :type Gameplay%20Tag%20Container%20Structure%20%28by%20ref%29) (Momentum :type Vector :default "0, 0, 0") (HitInfo :type Hit%20Result%20Structure%20%28by%20ref%29) (InstigatedBy :type Controller%20Object%20Reference) (DamageCauser :type Actor%20Object%20Reference) (EffectContext :type Gameplay%20Effect%20Context%20Handle%20Structure))
  ;[1 1 1] EVENT -> On Death
  (Collision|SetCollisionEnabled (Variables|Default|GetSM_SneakySnowman) "NoCollision")
  (CallFunction|ToggleFakeSnowmanMulticast false))

(event Custom|FrequentChecksTimer
  (CallFunction|CrouchedCheck)
  (CallFunction|PositionCheck))

(event Custom|PositionCheck
  (bind _playerpawn (Variables|StoredActors|GetPlayerPawn))
  (bind _returnvalue (Transformation|GetActorLocation _playerpawn))
  (bind _fakeactor (Variables|StoredActors|GetFakeActor))
  (bind _returnvalue_2 (Transformation|GetActorLocation _fakeactor))
  (bind _returnvalue_3 (Math|Vector|Distance2D(Vector) _returnvalue _returnvalue_2))
  (bind _maxplayerfakedistance (Variables|DesignVariables|GetMaxPlayerFakeDistance))
  (bind _returnvalue_4 (ScalableFloat|GetValueatLevel _maxplayerfakedistance))
  ;[1 1 1] If the player is too far from the fake while replaced by it, some physical pushing happened.
  (Networking|SwitchHasAuthority
    (:Authority
      (if (Math|Boolean|ANDBoolean (Variables|StateTrackers|GetCurrentlyReplacedByFake) (Variables|StateTrackers|GetAttached) (> _returnvalue_4 0))
        (if (or (< (* _returnvalue_4 2.0) _returnvalue_3) (CallFunction|IsFakeVerticalDistanceValid))
          _
          (elif (and (< _returnvalue_4 _returnvalue_3) (CallFunction|FinalPositionCheck))
            _
            (else
              (CallFunction|UpdateSnowmanState))))
        (CallFunction|OnInvalidInteractions false true)))))

(event Custom|SpawnProp
  (params (PlayMontage :type Boolean :default true) (RemoveTags :type Boolean :default true))
  (bind (_hasmovementbase _movementbase) (CallFunction|GetPlayerMovementBase))
  (bind (_hasmovementbase_2 _movementbase_2) (CallFunction|GetPlayerMovementBase))
  (bind _playerpawn (Variables|StoredActors|GetPlayerPawn))
  (Variables|CachedValues|SetRemoveTags RemoveTags)
  (Variables|StateTrackers|SetIsBeingReplacedWithProp true)
  ; Play JumpOut FX
  (Networking|CanExecuteCosmeticEvents
    (:True
      (Effects|Components|ParticleSystem|SpawnFXAtLocation (Variables|Constants|FX|GetP_JumpingOut) (Transformation|GetActorLocation))
      (Audio|PlaySoundatLocation (Variables|Sound|GetSoundOnJumpOut) (Transformation|GetActorLocation) "0, 0, 0" 1.0 1.0 0.0 0 0 self)
      (Networking|ForceNetUpdate))
    (:False
      _))
  ;[1 1 1] Spawn Snowman Prop in it's place (on jump out)
  (Variables|Default|SetbSwimmingOnExitProp (Variables|StateTrackers|GetSwimming))
  (if _hasmovementbase_2
    (Transformation|AttachActorToActor _movementbase_2 "None" "KeepWorld" "KeepWorld" "KeepWorld" false)
    (Variables|Default|SetMovingPlatformRelative (Variables|/Script/Engine.SceneComponent|GetRelativeLocation (Variables|Default|GetScene)))
    (else
      (Transformation|DetachFromActor "KeepWorld" "KeepWorld" "KeepWorld")))
  (if PlayMontage
    (Animation|Montage|PlayMontage (Variables|/Script/Engine.Character|GetMesh (Variables|StoredActors|GetPlayerPawn)) "/Game/Animation/Game/MainPlayer/Reacts/WorldInteraction/Hideable/Reappearing_Montage.Reappearing_Montage"
      (:then
        _)
      (:OnCompleted
        (Weapon|SetHolsterWeapon (Variables|StoredActors|GetPlayerPawn) 0 false false "None"))))
  (Default|CallOnJumpOut)
  (bind (_outhit _returnvalue) (Collision|LineTraceByChannel (Transformation|GetActorLocation) (- (Transformation|GetActorLocation) "0, 0,600.000000") "TraceTypeQuery1" false (Utilities|Array|MakeArray (Variables|StoredActors|GetPlayerPawn))))
  (Variables|CachedValues|SetDrop_GoalPos (CallFunction|GetRelativeLocationifAttached self (Math|Vector|SelectVector (+ (.Location (Collision|BreakHitResult _outhit)) "0, 0,20.000000") (.TraceEnd (Collision|BreakHitResult _outhit)) _returnvalue)))
  (if (Variables|StateTrackers|GetCurrentlyReplacedByFake)
    (CallFunction|SetPlayerHidden false)
    (Utilities|FlowControl|Delay (Components|Timeline|GetTimelineLength (Variables|Athena_Player_SneakySnowmanV2|GetDropToFloor))) ; Same amount of time as DropToFloor timeline length
    (Utilities|CalculateDropGoalRot (Variables|StoredActors|GetFakeActor))
    (CallFunction|UpdateActorLocationandRotation self (+ (CallFunction|GetCrouchedLoc) "0, 0,30.000000") (Transformation|GetActorRotation (Variables|StoredActors|GetFakeActor)))
    (CallFunction|ToggleFakeSnowmanMulticast false)
    (else
      ;;[0.8021 0.8021 0.8021] Interpolate snowman to exact position and rotation to match where we're spawning the prop.
      (Utilities|CalculateDropGoalRot self)
      (Variables|CachedValues|SetDrop_StartPos (CallFunction|GetRelativeLocationifAttached self (Transformation|GetActorLocation)))
      (bind (_direction _pos _rot) (Timeline|DropToFloor Play)
        (:Update
          (CallFunction|UpdateActorLocationandRotation self (Math|Vector|Lerp(Vector) (Variables|CachedValues|GetDrop_GoalPos) (Variables|CachedValues|GetDrop_StartPos) _pos) (Math|Rotator|Lerp(Rotator) (Variables|CachedValues|GetDrop_GoalRot) (Transformation|GetActorRotation) _rot true)))
        (:Finished
          _)) ;  
    ))
  (Utilities|FlowControl|Sequence
    (:then_0
      ;;[0.6771 0.6771 0.6771] Destroy this actor
      (Variables|StateTrackers|SetShouldSpawnDeathFX false) ; Don't spawn death fx since we're replacing it.
      (if (and (Weapon|IsWeaponHolstered _playerpawn) (Networking|IsServer))
        (Weapon|SetHolsterWeapon _playerpawn 0 false false "None")
        (elif (Variables|CachedValues|GetRemoveTags)
          (Ability|SendGameplayEventtoActor (Variables|Actor|GetInstigator) "(TagName=\"Abilities.Generic.ActiveAbility.Activate\")")
          (Actor|DestroyActor))))
    (:then_1
      ;[0.7188 0.7188 0.7188] Spawn the prop version in its place
      (Networking|SwitchHasAuthority
        (:Authority
          (if (not (Variables|Default|GetbSwimmingOnExitProp))
            (Utilities|CheckforMovableObject
              (:On%20Movable%20Object
                ;;[1 0.2774 0] If there's a vehicle under, do not spawn Snowman. Prevents issues like FORT-428499
                (if (Math|Boolean|ANDBoolean (CallFunction|AttachestoMovingPlatform) (CallFunction|GetPlayerMovementBase) (not (ScalableFloat|GetValueasBool (Variables|DesignVariables|GetCanCrouchOnMovingPlatform))))
                  _))
              (:No%20Movable%20Object
                (if (CallFunction|GetPlayerMovementBase)
                  (if (CallFunction|AttachestoMovingPlatform)
                    _))
                (bind _returnvalue_2 (Building|SpawnBuildingActor (Variables|Constants|GetPropToSpawnOnExit) (Math|Transform|MakeTransform (Math|Vector|SelectVector (Math|Transform|TransformLocation (Transformation|GetActorTransform) (+ (Variables|Constants|GetPropPosOffset) (Variables|CachedValues|GetDrop_GoalPos))) (+ (Variables|Constants|GetPropPosOffset) (Variables|CachedValues|GetDrop_GoalPos)) (Utilities|IsValid (Actor|GetAttachParentActor) :!class /Script/Engine.KismetSystemLibrary)) (Variables|CachedValues|GetDrop_GoalRot) (Transformation|GetActorScale3D)) 0 (Variables|StoredActors|GetPlayerPawn)))
                (if _hasmovementbase
                  (Transformation|AttachActorToActor _returnvalue_2 _movementbase "None" "KeepWorld" "KeepWorld" "KeepWorld" false)
                  (Transformation|SetActorRelativeLocation _returnvalue_2 (Variables|Default|GetMovingPlatformRelative) false false)
                  (Actor|SetLifeSpan _returnvalue_2 (ScalableFloat|GetValueatLevel (Variables|DesignVariables|GetMovingPlatformLifespan)))
                  (else
                    _))
                (Default|CallOnSpawnProp _returnvalue_2)
                (Ability|SendGameplayEventtoActor (Variables|StoredActors|GetPlayerPawn) "(TagName=\"Abilities.Generic.TriggeredAbility.Activate\")" (Utilities|Struct|MakeGameplayEventData :Target _returnvalue_2 :ContextHandle "()" :InstigatorTags "(GameplayTags=)" :TargetTags "(GameplayTags=)" :EventMagnitude 0.0 :TargetData "()"))
                (bind _asathena_prop_cool_carpet (Utilities|Casting|CastToAthena_Prop_CoolCarpet _returnvalue_2)) ; If prop spawned was a cardboard box, close the flaps
                (Class|AthenaPropCoolCarpet|CloseFlaps _asathena_prop_cool_carpet)))))))))

(event Custom|CrouchedCheck
  (bind _returnvalue (Transformation|GetActorLocation))
  ;[1 1 1] Crouched Check
  (if (Variables|/Script/Engine.Character|GetbIsCrouched (Variables|StoredActors|GetPlayerPawn))
    (Collision|SetCollisionEnabled (Variables|Default|GetCylinder) "QueryOnly")
    (if (not (Variables|StateTrackers|GetCurrentlyCrouching))
      (Variables|StateTrackers|SetCurrentlyCrouching true)
      (Audio|PlaySoundatLocation (Variables|Sound|GetSoundOnCrouch) _returnvalue))
    (else
      (Collision|SetCollisionEnabled (Variables|Default|GetCylinder) "NoCollision")
      (if (Variables|StateTrackers|GetCurrentlyCrouching)
        (Variables|StateTrackers|SetCurrentlyCrouching false)
        (Audio|PlaySoundatLocation (Variables|Sound|GetSoundOnStand) _returnvalue)))))

(event Custom|OnInteractWithFake
  (params (InteractingPawn :type Fort%20Pawn%20Object%20Reference))
  ;[1 1 1] If another player interacts with it while crouched, destroy the snowman
  (if (not (Utilities|Equal(Object) InteractingPawn (Variables|StoredActors|GetPlayerPawn)))
    (Variables|StateTrackers|SetPlayerIsBeingForcedOut true)
    (CallFunction|ToggleFakeSnowmanMulticast false)
    (Ability|SendGameplayEventtoActor (Variables|Actor|GetInstigator) "(TagName=\"Abilities.Generic.ActiveAbility.Activate\")")
    (Actor|DestroyActor)))

(event Custom|Destroy
  ;[1 1 1] Destroy
  (CallFunction|ToggleFakeSnowmanMulticast false)
  (Ability|SendGameplayEventtoActor (Variables|Actor|GetInstigator) "(TagName=\"Abilities.Generic.ActiveAbility.Activate\")")
  (Actor|DestroyActor))

(event Custom|StopHiding
  (params (LaunchPlayer :type Boolean :default true))
  (bind (_origin _boxextent) (Collision|GetComponentBounds (Variables|Default|GetSM_SneakySnowman)))
  ;[1 1 1] Stop Hiding Logic
  (Variables|StateTrackers|SetAttached false)
  (if (not (and (GameplayTags|HasMatchingGameplayTag (Variables|StoredActors|GetPlayerPawn) "(TagName=\"Abilities.Generic.Athena.ZipLine\")") LaunchPlayer))
    (Character|LaunchCharacter (Variables|StoredActors|GetPlayerPawn) (Math|Vector|vector+vector (* (Transformation|GetActorForwardVector (Variables|StoredActors|GetPlayerPawn)) (Variables|DesignVariables|GetJumpOut_LateralSpeed)) (Math|Vector|MakeVector 0.0 0.0 (Variables|DesignVariables|GetJumpOut_VerticalSpeed))) true true) ; Launch Player standing on prop in random direction
  )
  (GameplayEffects|RemoveActiveEffectswithTags (Ability|GetAbilitySystemComponent (Variables|StoredActors|GetPlayerPawn)) "(GameplayTags=((TagName=\"Athena.Item.Snowman\")))")
  (Networking|SwitchHasAuthority
    (:Authority
      (bind (_outactors _returnvalue) (Collision|BoxOverlapActors _origin _boxextent (Utilities|Array|MakeArray "ObjectTypeQuery1") "/Script/FortniteGame.FortWaterBodyActor" :!class /Script/Engine.KismetSystemLibrary))
      (if _returnvalue
        (CallFunction|Destroy)
        (else
          (CallFunction|SpawnProp true))))))

(event Custom|Client_SetHolster
  (params (ShouldHolster :type Boolean))
  (Weapon|SetHolsterWeapon (Variables|StoredActors|GetPlayerPawn) 0 ShouldHolster false (Variables|Default|GetSelfHolsterID)))

(event Custom|SetPlayerHidden
  (params (PlayerHidden :type Boolean))
  (bind _playerpawn (Variables|StoredActors|GetPlayerPawn))
  (bind _returnvalue (Weapon|GetCurrentWeapon _playerpawn))
  (bind _playerpawn_2 (Variables|StoredActors|GetPlayerPawn))
  ;[1 1 1] Start Hiding Logic
  ;;[0.9763 1 0.3523] Swapped HiddenWithTag to use new  IsHiddingInProp function
  (Rendering|SetActorHiddenInGameWithTag (Variables|StoredActors|GetPlayerPawn) PlayerHidden "(TagName=\"Pawn.HidePlayerPawn.SneakySnowman\")" 0 true)
  (Networking|SwitchHasAuthority
    (:Authority
      (bind _asplayer_pawn_athena (Utilities|Casting|CastToPlayerPawn_Athena (Variables|StoredActors|GetPlayerPawn)))
      (Class|PlayerPawnAthena|IsHiddinginProp _asplayer_pawn_athena PlayerHidden)
      (Variables|StateTrackers|SetPlayerHidden PlayerHidden))
    (:Remote
      (FortPawn|SetPawnVisibilityLocal (Variables|StoredActors|GetPlayerPawn) PlayerHidden true)
      ;;[1 0.3594 0.3594] Touch controls : Show throw trajectory when hidden on prop
      (if (and PlayerHidden (TouchPlatform|IsUsingTouch (Fortnite|GetFortPlayerControllerfromActor (Variables|StoredActors|GetPlayerPawn))))
        (FortPawn|BindEventtoOnWeaponEquippedDelegate _playerpawn_2 (delegate Custom|OnWeaponEquipped))
        (Ability|Tasks|WaitGameplayTagAddToActor _playerpawn "(TagName=\"Pawn.HidePlayerPawn.SneakySnowman\")" true
          (:then
            _)
          (:Added
            _))
        (Utilities|IsValid _returnvalue
          (:Is%20Valid
            (FortWeapon|SetTrajectoryVisibility _returnvalue true)))
        (else
          (FortPawn|UnbindEventfromOnWeaponEquippedDelegate _playerpawn_2 (delegate Custom|OnWeaponEquipped)))))))

(event Custom|OnWeaponEquipped
  (params (NewWeapon :type Fort%20Weapon%20Object%20Reference) (PrevWeapon :type Fort%20Weapon%20Object%20Reference))
  (bind _playerpawn (Variables|StoredActors|GetPlayerPawn))
  (bind _returnvalue (Weapon|GetCurrentWeapon _playerpawn))
  ;[1 1 1] Start Hiding Logic
  ;;[1 0.3594 0.3594] Touch controls : Show throw trajectory when hidden on prop
  ;;;[1 1 1] Process at the end of frame to override possible visibility set to false
  (Utilities|FlowControl|DelayUntilNextTick)
  (if (and (Utilities|IsValid _returnvalue :!class /Script/Engine.KismetSystemLibrary) (GameplayTags|HasMatchingGameplayTag _playerpawn "(TagName=\"Pawn.HidePlayerPawn.SneakySnowman\")"))
    (FortWeapon|SetTrajectoryVisibility _returnvalue true)))

(event Custom|SetSnowmanHidden
  (params (Hidden :type Boolean) (InstantOnAll :type Boolean :default true))
  ;[1 1 1] Start Hiding Logic
  (Networking|SwitchHasAuthority
    (:Authority
      (Variables|StateTrackers|SetSelfHidden Hidden))
    (:Remote
      (if (or InstantOnAll (Pawn|IsLocallyControlled (Variables|StoredActors|GetPlayerPawn)))
        (Rendering|SetActorHiddenInGame Hidden)))))

(fn AttachesToMovingPlatform
  (return (ScalableFloat|GetValueasBool (Variables|DesignVariables|GetCanAttachToMovingPlatform))))

(fn CheckForClipping_
  (bind (_origin _boxextent) (Collision|GetComponentBounds (Variables|Default|GetSM_SneakySnowman)))
  (bind (_outhit _returnvalue) (Collision|LineTraceByChannel (+ (Transformation|GetActorLocation) "0, 0,100.000000") (+ (Transformation|GetActorLocation) "0, 0,-1000.000000") "TraceTypeQuery10"))
  (if _returnvalue
    (bind (_outcomponents _returnvalue_2) (Collision|BoxOverlapComponents _origin (Math|Vector|MakeVector (.x _boxextent) (.y _boxextent) (* (.z _boxextent) 0.75)) (Utilities|Array|MakeArray "ObjectTypeQuery1") 0 (Utilities|Array|MakeArray self (.HitActor (Collision|BreakHitResult _outhit)) (Variables|StoredActors|GetPlayerPawn) (Variables|StoredActors|GetFakeActor) (Variables|StoredActors|GetInteractSphereRef)) :!class /Script/Engine.KismetSystemLibrary))
    (if _returnvalue_2
      (for _array_element _outcomponents
        (if (Utilities|Enum|Equal(Enum) (Collision|GetCollisionResponsetoChannel _array_element "ECC_Pawn") "ECR_Block")
          (return true))))
    (return false)
    (else
      (return true))))

(fn CheckIfCanJumpOut
  (params (InteractingPawn :type Object%20Reference))
  (return (and (Utilities|Equal(Object) InteractingPawn (Variables|StoredActors|GetPlayerPawn)) (Variables|StateTrackers|GetAttached))))

(fn DeleteOnResetMinigamePlayerIfAttached
  (return true))

(fn FinalPositionCheck
  (return true))

(fn FlattestHitResult
  (params (NewParam :type Array%20of%20Hit%20Result%20Structures))
  (for _array_element NewParam
    (if (> (Math|Vector|DotProduct (.ImpactNormal (Collision|BreakHitResult _array_element)) "0, 0,1.000000") (LocalVariables|GetBestDot))
      (LocalVariables|SetCurrentBest _array_element)
      (LocalVariables|SetBestDot (Math|Vector|DotProduct (.ImpactNormal (Collision|BreakHitResult _array_element)) "0, 0,1.000000"))))
  (return (LocalVariables|GetCurrentBest)))

(fn GetCrouchedLoc
  (bind (_origin _boxextent _sphereradius) (Collision|GetComponentBounds (Variables|/Script/Engine.Character|GetMesh (Variables|StoredActors|GetPlayerPawn))))
  (bind (_returnvalue_x _returnvalue_y) (Math|Vector|BreakVector (Math|Vector|vector+vector (Transformation|GetActorLocation (Variables|StoredActors|GetPlayerPawn)) (* (Transformation|GetActorForwardVector (Variables|StoredActors|GetPlayerPawn)) -12.0) (* (Transformation|GetActorRightVector (Variables|StoredActors|GetPlayerPawn)) -6.0))))
  (bind (_outhit _returnvalue) (Collision|LineTraceByChannel (+ (Transformation|GetActorLocation (Variables|StoredActors|GetPlayerPawn)) "0, 0,50.000000") (+ (Transformation|GetActorLocation (Variables|StoredActors|GetPlayerPawn)) "0, 0,-100.000000") "TraceTypeQuery10" false (Utilities|Array|MakeArray (Variables|StoredActors|GetPlayerPawn))))
  (bind (_bblockinghit _binitialoverlap _time _distance _location) (Collision|BreakHitResult _outhit))
  (bind (_location_x _location_y _location_z) (Math|Vector|BreakVector _location))
  (return (CallFunction|GetRelativeLocationifAttached (Variables|StoredActors|GetFakeActor) (Math|Vector|MakeVector _returnvalue_x _returnvalue_y (select (CallFunction|IsMeshHeightScalePositive) (select _returnvalue _location_z (- (.z _origin) _sphereradius)) (+ (* (.z _boxextent) 2.0) (select _returnvalue _location_z (- (.z _origin) _sphereradius))))))))

(fn GetFakeRotation
  (bind (_returnvalue_roll _returnvalue_pitch _returnvalue_yaw) (Math|Rotator|BreakRotator (Transformation|GetActorRotation)))
  (if (Variables|DesignVariables|GetRotateFakeToMatchGroundNormal)
    (Utilities|FlowControl|Sequence
      (:then_0
        (bind _outhit (Collision|LineTraceByChannel (+ (Transformation|GetActorLocation) (* (Transformation|GetActorForwardVector (Variables|StoredActors|GetPlayerPawn)) 3e+01)) (- (+ (Transformation|GetActorLocation) (* (Transformation|GetActorForwardVector (Variables|StoredActors|GetPlayerPawn)) 3e+01)) "0, 0,200.000000") "TraceTypeQuery1" false (Utilities|Array|MakeArray (Variables|StoredActors|GetPlayerPawn)))))
      (:then_1
        (bind _outhit_2 (Collision|LineTraceByChannel (+ (Transformation|GetActorLocation) (* (Transformation|GetActorRightVector (Variables|StoredActors|GetPlayerPawn)) 3e+01)) (- (+ (Transformation|GetActorLocation) (* (Transformation|GetActorRightVector (Variables|StoredActors|GetPlayerPawn)) 3e+01)) "0, 0,200.000000") "TraceTypeQuery1" false (Utilities|Array|MakeArray (Variables|StoredActors|GetPlayerPawn)) "None" true "(R=0.000000,G=0.942196,B=1.000000,A=1.000000)" "(R=1.000000,G=0.000000,B=0.841290,A=1.000000)")))
      (:then_2
        (bind _outhit_3 (Collision|LineTraceByChannel (Math|Vector|vector+vector (Transformation|GetActorLocation) (* (Transformation|GetActorRightVector (Variables|StoredActors|GetPlayerPawn)) -2e+01) (* (Transformation|GetActorForwardVector (Variables|StoredActors|GetPlayerPawn)) -2e+01)) (- (Math|Vector|vector+vector (Transformation|GetActorLocation) (* (Transformation|GetActorRightVector (Variables|StoredActors|GetPlayerPawn)) -2e+01) (* (Transformation|GetActorForwardVector (Variables|StoredActors|GetPlayerPawn)) -2e+01)) "0, 0,200.000000") "TraceTypeQuery1" false (Utilities|Array|MakeArray (Variables|StoredActors|GetPlayerPawn)) "None" true "(R=0.000000,G=0.942196,B=1.000000,A=1.000000)" "(R=1.000000,G=0.000000,B=0.841290,A=1.000000)"))
        (bind _flattest (CallFunction|FlattestHitResult (Utilities|Array|MakeArray _outhit _outhit_2 _outhit_3)))
        (if (not (< (Math|Vector|DotProduct (.Normal (Collision|BreakHitResult _flattest)) "0, 0,1.000000") 0.5))
          (return (Math|Rotator|MakeRotfromXZ (Math|Vector|Normalize (Math|Vector|ProjectVectorontoPlane (Transformation|GetActorForwardVector) (.Normal (Collision|BreakHitResult _flattest)))) (.Normal (Collision|BreakHitResult _flattest)))))))
    (else
      _))
  (return 0.0 0.0 _returnvalue_yaw))

(fn GetPlayerMovementBase
  (bind _asprimitive_component (Utilities|Casting|CastToPrimitiveComponent (Pawn|Components|CharacterMovement|GetMovementBaseObject (Variables|/Script/Engine.Character|GetCharacterMovement (Variables|StoredActors|GetPlayerPawn))))
    (:then
      (Utilities|IsValid (Components|GetOwner _asprimitive_component)
        (:Is%20Valid
          (return (Utilities|IsValid (Class|FortBaseMovingPlatformComponent|FindMovingPlatformComponentfromActor (Components|GetOwner _asprimitive_component) true) :!class /Script/Engine.KismetSystemLibrary) (Components|GetOwner _asprimitive_component)))
        (:Is%20Not%20Valid
          _)))
    (:CastFailed
      _))
  (return false 0))

(fn GetRelativeLocationIfAttached
  (params (TargetActor :type Actor%20Object%20Reference) (WorldLocation :type Vector))
  (bind _returnvalue (Actor|GetAttachParentActor TargetActor))
  (Utilities|IsValid _returnvalue
    (:Is%20Valid
      (return (Math|Transform|InverseTransformLocation (Transformation|GetActorTransform _returnvalue) WorldLocation) true))
    (:Is%20Not%20Valid
      (return WorldLocation false))))

(fn IsFakeVerticalDistanceValid
  (bind _maxplayerfakeverticaldistance (Variables|Default|GetMaxPlayerFakeVerticalDistance))
  (bind _returnvalue (ScalableFloat|GetValueatLevel _maxplayerfakeverticaldistance))
  (if (> _returnvalue 0)
    (return (> (Math|Float|Absolute(Float) (- (.z (Transformation|GetActorLocation (Variables|StoredActors|GetPlayerPawn))) (.z (Transformation|GetActorLocation (Variables|StoredActors|GetFakeActor))))) _returnvalue))
    (else
      (return false))))

(fn IsMeshHeightScalePositive
  (bind (_returnvalue_x _returnvalue_y _returnvalue_z) (Math|Vector|BreakVector (Transformation|GetWorldScale (Variables|Default|GetSM_SneakySnowman))))
  (return (> _returnvalue_z 0.0)))

(fn OnBeginInteractWithSphere
  (Variables|StateTrackers|SetInteracting true)
  (Utilities|ApplyEffect "/Game/Athena/Items/Consumables/SneakySnowman/CoolCarpet/GE_BlockADS.GE_BlockADS_C" self))

(fn OnInterruptInteractWithSphere
  (Variables|StateTrackers|SetInteracting false))

(fn OnRep_PlayerHidden)

(fn OnRep_SelfHidden
  (Rendering|SetActorHiddenInGame (Variables|StateTrackers|GetSelfHidden)))

(fn OnStartedUsingFake)

(fn OnStoppedUsingFake)

(fn OnWeaponChange
  (params (NewWeapon :type Fort%20Weapon%20Object%20Reference))
  (if (CallFunction|ShouldDestroywhenUsingWeapon NewWeapon)
    (bind _playerpawn (Variables|StoredActors|GetPlayerPawn)
      (:then
        (Utilities|IsValid (Weapon|GetCurrentWeapon _playerpawn)
          (:Is%20Valid
            _)
          (:Is%20Not%20Valid
            _))
        (CallFunction|OnInvalidInteractions false true))
      (:else))))

(fn ShouldDestroyWhenUsingWeapon
  (params (Weapon :type Fort%20Weapon%20Object%20Reference))
  (return (and (Variables|Default|GetDestroyWhenSwappingOrShooting) (not (GameplayTags|HasMatchingGameplayTag (Data|GetWeaponData Weapon) "(TagName=\"Item.Consumable.SneakySnowman\")")))))

(fn ShouldShowFake
  (return (Math|Boolean|ANDBoolean (Math|Boolean|NotEqual(Boolean) (or (and (AI|Components|NavMovement|IsCrouching (Variables|/Script/Engine.Character|GetCharacterMovement (Variables|StoredActors|GetPlayerPawn))) (or (< (Math|Vector|VectorLength (Variables|/Script/Engine.MovementComponent|GetVelocity (Variables|/Script/Engine.Character|GetCharacterMovement (Variables|StoredActors|GetPlayerPawn)))) 1.0) (Variables|Constants|GetAllowFakeWhileMoving))) (Variables|Default|GetForceKeepFakeOn)) (Variables|StateTrackers|GetCurrentlyReplacedByFake)) (or (GameplayTags|HasMatchingGameplayTag (Variables|StoredActors|GetPlayerPawn) "(TagName=\"MovementMode.Falling\")") (not (Variables|StateTrackers|GetIsPoppingOut))) (not (Variables|StateTrackers|GetPlayerIsBeingForcedOut))) (or (and (AI|Components|NavMovement|IsCrouching (Variables|/Script/Engine.Character|GetCharacterMovement (Variables|StoredActors|GetPlayerPawn))) (or (< (Math|Vector|VectorLength (Variables|/Script/Engine.MovementComponent|GetVelocity (Variables|/Script/Engine.Character|GetCharacterMovement (Variables|StoredActors|GetPlayerPawn)))) 1.0) (Variables|Constants|GetAllowFakeWhileMoving))) (Variables|Default|GetForceKeepFakeOn))))

(fn UpdateActorLocationAndRotation
  (params (TargetActor :type Actor%20Object%20Reference) (Location :type Vector) (Rotation :type Rotator))
  (Transformation|SetActorRotation TargetActor Rotation false)
  (Utilities|IsValid (Actor|GetAttachParentActor TargetActor)
    (:Is%20Valid
      (Transformation|SetActorRelativeLocation TargetActor Location false false))
    (:Is%20Not%20Valid
      (Transformation|SetActorLocation TargetActor Location false false))))

(fn UpdateHolster
  (params (ShouldHolster :type Boolean))
  (Weapon|SetHolsterWeapon (Variables|StoredActors|GetPlayerPawn) 0 ShouldHolster false (Variables|Default|GetSelfHolsterID))
  (Networking|SwitchHasAuthority
    (:Authority
      (CallFunction|ClientSetHolster ShouldHolster))))

(fn UpdateSnowmanState
  (bind (_shouldtoggle _usefake) (CallFunction|ShouldShowFake))
  (if (and (Variables|StateTrackers|GetAttached) (Variables|DesignVariables|GetShouldMaintainRotationWhileHiding))
    (if _shouldtoggle
      (Utilities|CheckforMovableObject
        (:On%20Movable%20Object
          (if _usefake
            (Utilities|FlowControl|DoOnce false
              (:Completed
                (CallFunction|OnInvalidInteractions false true)))))
        (:No%20Movable%20Object
          (CallFunction|ToggleFakeSnowmanMulticast _usefake (CallFunction|GetCrouchedLoc) true false true))))))

(fn ConstructionScript
  (CallFunction|SetTeam (Team|GetActorTeam (Game|GetInstigator))))
