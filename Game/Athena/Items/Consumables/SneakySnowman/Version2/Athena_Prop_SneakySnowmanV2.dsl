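; Sneaky Snowman V2 prop: a placeable snowman that players interact with to hide inside.
; On begin play it wobbles, validates its spot, watches supporting building actors for destruction and applies water nearby via Curie.
; Interaction checks line of sight and tags, applies effects and sends the Hide ability event, while a timeline lerps the player in, attaching to any moving base.
; Weapon kills can drop a held snowball; a timer failsafe destroys the prop.
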(event EventBeginPlay
  (bind (_origin _boxextent) (Collision|GetComponentBounds (Variables|Debug|GetStaticMeshComponent)))
  (bind (_origin_2 _boxextent_2) (Collision|GetComponentBounds (Variables|Debug|GetStaticMeshComponent)))
  ;[1 1 1] EVENT -> BeginPlay: Double check if gamestate has not loaded then we will turn off projectile movement. This is to fix Snowmen from falling through the world
  (Utilities|FlowControl|Sequence
    (:then_1
      (Networking|SwitchHasAuthority
        (:Authority
          (bind _asfort_game_state_athena (Utilities|Casting|CastToFortGameStateAthena (Game|GetGameState)))
          (if (Variables|/Script/FortniteGame.FortGameStateAthena|GetbPlaylistDataIsLoaded _asfort_game_state_athena)
            (Variables|Default|SetGamestateLoaded true)
            (else
              (Playlist|BindEventtoOnCurrentPlaylistLoaded _asfort_game_state_athena (delegate Custom|PlaylistLoaded)))))))
    (:then_2
      (CallFunction|WobbleAnims))
    (:then_3
      ;[1 1 1] if the props in world are sitting on a destructable object, destroy it when it's destroyed
      (bind _outcomponents (Collision|BoxOverlapComponents (- (Transformation|GetActorLocation) "0.000000,0.000000,20.000000") (* _boxextent_2 "0.600000,0.600000,0.600000") (Variables|Default|GetObjectTypes) 0 (Utilities|Array|MakeArray self) :!class /Script/Engine.KismetSystemLibrary))
      (bind _array_element (Utilities|Array|ForEachLoopwithBreak _outcomponents)
        (:LoopBody
          (bind _asbuilding_actor (Utilities|Casting|CastToBuildingActor (Components|GetOwner _array_element)))
          (if (Variables|/Script/Engine.Actor|GetbCanBeDamaged _asbuilding_actor)
            (EventDispatchers|AssignOnDied _asbuilding_actor (delegate Custom|WorldPropSupportDied)) ; Break after finding the first one
          ))))
    (:then_4
      ;[1 1 1] Validating Spawn Location
      (Utilities|IsValid (Variables|Actor|GetInstigator)
        (:Is%20Valid
          (bind (_outhit _returnvalue) (Collision|LineTraceByChannel (+ (Transformation|GetActorLocation) "0, 0,100.000000") (+ (Transformation|GetActorLocation) "0, 0,-1000.000000") (select (ScalableFloat|GetValueasBool (Variables|Default|GetUseWorldAndPawnMultiTraceForPropDestructionCheck)) (Utilities|Enum|LiteralenumETraceTypeQuery "TraceTypeQuery13") (Utilities|Enum|LiteralenumETraceTypeQuery "TraceTypeQuery10"))))
          (if _returnvalue
            (Transformation|SetActorLocation (.ImpactPoint (Collision|BreakHitResult _outhit)) false true)
            (bind (_outcomponents_2 _returnvalue_2) (Collision|BoxOverlapComponents (+ (Variables|Default|GetBoxOverlapCheckSpawnOffset) _origin) (Math|Vector|MakeVector (.x _boxextent) (.y _boxextent) (* (.z _boxextent) 0.75)) (Utilities|Array|MakeArray "ObjectTypeQuery1" "ObjectTypeQuery2") 0 (Utilities|Array|MakeArray self (select (.bSuccess (Utilities|Casting|CastToCar_DEFAULT (.HitActor (Collision|BreakHitResult _outhit)))) 0 (.HitActor (Collision|BreakHitResult _outhit)))) :!class /Script/Engine.KismetSystemLibrary))
            (if _returnvalue_2
              (for _array_element_2 _outcomponents_2
                (if (Class|AthenaFunctionLibrary|RowtoBool (Variables|Default|GetShouldIgnoreSplinesOnSpawnCheck))
                  ;;[1 1 1] To fix the issue: FORT-685598
                  ;; that is caused manly by spline rotations that prevent the sneaky snowman to spawn since box overlap check doesn't support rotation
                  (Utilities|Casting|CastToSplineMeshComponent _array_element_2
                    (:CastFailed
                      (Utilities|Casting|CastToControlPointMeshComponent _array_element_2
                        (:CastFailed
                          _)))))
                (if (Utilities|Enum|Equal(Enum) (Collision|GetCollisionResponsetoChannel _array_element_2 "ECC_Pawn") "ECR_Block")
                  (Actor|DestroyActor))))
            (else
              (CallFunction|OnSpawnedNotonPawn))))) ; Null if this was placed by level designers, in which case we don't need to validate the location.
    )
    (:then_5
      (CallFunction|ApplyWaterinRadius))))

(event Custom|PlaylistLoaded
  (params (PlaylistName :type Name) (PlaylistContextTags :type Gameplay%20Tag%20Container%20Structure%20%28by%20ref%29))
  (bind (_origin _boxextent) (Collision|GetComponentBounds (Variables|Debug|GetStaticMeshComponent)))
  ;[1 1 1] EVENT -> BeginPlay: Double check if gamestate has not loaded then we will turn off projectile movement. This is to fix Snowmen from falling through the world
  ;[1 1 1] if the props in world are sitting on a destructable object, destroy it when it's destroyed
  (bind _outcomponents (Collision|BoxOverlapComponents (- (Transformation|GetActorLocation) "0.000000,0.000000,20.000000") (* _boxextent "0.600000,0.600000,0.600000") (Variables|Default|GetObjectTypes) 0 (Utilities|Array|MakeArray self) :!class /Script/Engine.KismetSystemLibrary))
  (bind _array_element (Utilities|Array|ForEachLoopwithBreak _outcomponents)
    (:LoopBody
      (bind _asbuilding_actor (Utilities|Casting|CastToBuildingActor (Components|GetOwner _array_element)))
      (if (Variables|/Script/Engine.Actor|GetbCanBeDamaged _asbuilding_actor)
        (EventDispatchers|AssignOnDied _asbuilding_actor (delegate Custom|WorldPropSupportDied)) ; Break after finding the first one
      ))))

(event Custom|WorldPropSupportDied
  (params (DamagedActor :type Actor%20Object%20Reference) (Damage :type Float%20%28single-precision%29) (InstigatedBy :type Controller%20Object%20Reference) (DamageCauser :type Actor%20Object%20Reference) (HitLocation :type Vector) (FHitComponent :type Primitive%20Component%20Object%20Reference) (BoneName :type Name) (Momentum :type Vector))
  ;[1 1 1] if the props in world are sitting on a destructable object, destroy it when it's destroyed
  (Actor|DestroyActor))

(event EventOnDeathServer
  (params (Damage :type Float%20%28single-precision%29) (DamageTags :type Gameplay%20Tag%20Container%20Structure%20%28by%20ref%29) (Momentum :type Vector :default "0, 0, 0") (HitInfo :type Hit%20Result%20Structure%20%28by%20ref%29) (InstigatedBy :type Controller%20Object%20Reference) (DamageCauser :type Actor%20Object%20Reference) (EffectContext :type Gameplay%20Effect%20Context%20Handle%20Structure))
  (bind (_returnvalue_location _returnvalue_rotation _returnvalue_scale) (Math|Transform|BreakTransform (Transformation|GetActorTransform)))
  ;[1 1 1] Spawn Snowball
  (if (not (Variables|Default|GetDead))
    (Variables|Default|SetDead true)
    (CallFunction|DisableCollision)
    (if (Math|Boolean|ANDBoolean (> (Math|Vector|VectorLength Momentum) 0.0) (Class|AthenaFunctionLibrary|RowtoBool (Variables|Default|GetShouldSpawnSnowball)) (GameplayTags|HasTag DamageTags "(TagName=\"Weapon\")"))
      (bind _returnvalue (Building|SpawnBuildingGameplayActor "/Game/Athena/Items/HeldObjects/Snowball/B_HeldObject_Snowball.B_HeldObject_Snowball_C" (Math|Transform|MakeTransform (+ _returnvalue_location "0, 0,125.000000") _returnvalue_rotation _returnvalue_scale) InstigatedBy false false))
      (bind _asb_held_object_snowball (Utilities|Casting|CastToB_HeldObject_Snowball _returnvalue))
      (Class|BHeldObjectSnowball|PlayDropAnimation _asb_held_object_snowball))))

(event EventOnDamagePlayEffects
  (params (Damage :type Float%20%28single-precision%29) (DamageTags :type Gameplay%20Tag%20Container%20Structure%20%28by%20ref%29) (Momentum :type Vector :default "0, 0, 0") (HitInfo :type Hit%20Result%20Structure%20%28by%20ref%29) (InstigatedBy :type Fort%20Pawn%20Object%20Reference) (DamageCauser :type Actor%20Object%20Reference) (EffectContext :type Gameplay%20Effect%20Context%20Handle%20Structure))
  ;[1 1 1] EVENT -> On Damage Play Effects: Trigger wobble anims on damage
  (bind _returnvalue (CallFunction|GetorCreateMIDOnDefaultMesh 0))
  (Rendering|Material|SetScalarParameterValue _returnvalue "WobbleStartTime" (Utilities|Time|GetGameTimeinSeconds) :!class /Script/Engine.MaterialInstanceDynamic)
  (Rendering|Material|SetScalarParameterValue _returnvalue "WobbleStrength" 0.25 :!class /Script/Engine.MaterialInstanceDynamic)
  (Rendering|Material|SetScalarParameterValue _returnvalue "WobbleDuration" 0.6 :!class /Script/Engine.MaterialInstanceDynamic))

(event OnComponentBeginOverlap(OverlapCheck)
  (params (OverlappedComponent :type Primitive%20Component%20Object%20Reference) (OtherActor :type Actor%20Object%20Reference) (OtherComp :type Primitive%20Component%20Object%20Reference) (OtherBodyIndex :type Integer) (bFromSweep :type Boolean) (SweepResult :type Hit%20Result%20Structure%20%28by%20ref%29))
  ;[1 1 1] EVENT -> OverlapCheck: Destroy if we are overlapping pawn or water on spawn
  (Utilities|Casting|CastToFortPlayerPawn OtherActor)
  (Actor|DestroyActor))

(event Custom|WobbleAnims
  ;[1 1 1] Trigger Wobble Anims on Spawn
  (if (Variables|Default|GetGamestateLoaded)
    (bind _returnvalue (CallFunction|GetorCreateMIDOnDefaultMesh 0))
    (Rendering|Material|SetScalarParameterValue _returnvalue "WobbleStartTime" (Utilities|Time|GetGameTimeinSeconds) :!class /Script/Engine.MaterialInstanceDynamic)
    (Rendering|Material|SetScalarParameterValue _returnvalue "WobbleStrength" 1.0 :!class /Script/Engine.MaterialInstanceDynamic)
    (Rendering|Material|SetScalarParameterValue _returnvalue "WobbleDuration" 1.0 :!class /Script/Engine.MaterialInstanceDynamic)
    ;[1 1 1] Destroy Overlap component after delay because it's no longer needed
    (Utilities|FlowControl|Delay 0.5)
    (Components|DestroyComponent (Variables|Default|GetOverlapCheck))))

(event Interaction|EventBlueprintOnInteract
  (params (InteractingPawn :type Fort%20Pawn%20Object%20Reference) (InteractionBeingAttempted :type EInteractionBeingAttempted%20Enum :default "FirstInteraction"))
  ;[1 1 1] Interaction Logic (Jumping In)
  (if (Interaction|BlueprintCanInteract InteractingPawn InteractionBeingAttempted)
    (Analytics|FireAnalyticEventEnvironmentalItemUsed (Utilities|Casting|CastToFortPlayerController (Pawn|GetController InteractingPawn)) self 0.0 "PlayerExit")
    (CallFunction|DisableCollision)
    (Utilities|ApplyEffect (Variables|Default|GetApplicationEffect) InteractingPawn)
    (Utilities|ApplyEffect "/Game/Athena/Items/Consumables/SneakySnowman/Version2/GE_Athena_BlockAllHideables.GE_Athena_BlockAllHideables_C" InteractingPawn)
    (Utilities|ApplyEffect "/Game/Athena/Items/Consumables/SneakySnowman/Version3/GE_Athena_VisibleToAI.GE_Athena_VisibleToAI_C" InteractingPawn)
    (Utilities|FlowControl|Delay 0.2)
    (Ability|SendGameplayEventtoActor InteractingPawn "(TagName=\"Abilities.Generic.Snowman.Hide\")" (Utilities|Struct|MakeGameplayEventData "(TagName=\"Abilities.Generic.Snowman.Hide\")" InteractingPawn 0 InteractingPawn 0 "()" "(GameplayTags=)" "(GameplayTags=)" 0.2))))

(event EventMovePlayer
  (params (Player :type Actor%20Object%20Reference))
  (bind _returnvalue (Actor|GetAttachParentActor))
  (bind _endlocation (Variables|Default|GetEndLocation))
  (bind _x (.x _endlocation))
  (bind _y (.y _endlocation))
  (bind _z (.z _endlocation))
  (bind _returnvalue_2 (+ _z 8e+01))
  (bind _startlocation (Variables|Default|GetStartLocation))
  ;[1 1 1] Move player to snowman on jump in
  (CallFunction|DetachifAttachedActorforMoveIn false) ; Detach if somehow 2 players entered the same snowman
  (Utilities|IsValid _returnvalue
    (:Is%20Valid
      ;;[0.9033 1 0.4002] For safety we'll call the detach function on end play. So we cache the actor we attached and the base we attached it to.
      (Variables|Default|SetActorLerpedForMoveIn Player)
      (bind _output_get (Variables|Default|SetAttachedBase _returnvalue))
      (Transformation|AttachActorToActor Player _output_get "None" "KeepWorld" "KeepWorld" "KeepWorld" false))
    (:Is%20Not%20Valid
      _))
  (bind _asfort_player_pawn (Utilities|Casting|CastToFortPlayerPawn Player)
    (:then
      (FortMovement|TemporarilyIgnoreClientMovementErrorChecksandCorrection _asfort_player_pawn (Components|Timeline|GetTimelineLength (Variables|Athena_Prop_SneakySnowmanV2|GetTimeline))))
    (:CastFailed
      _))
  (Utilities|IsValid (Actor|GetAttachParentActor)
    (:Is%20Valid
      (Variables|Default|SetStartLocation (Math|Transform|InverseTransformLocation (Transformation|GetActorTransform _returnvalue) (Transformation|GetActorLocation Player)))
      (Variables|Default|SetEndLocation (Math|Transform|InverseTransformLocation (Transformation|GetActorTransform _returnvalue) (Transformation|GetActorLocation))))
    (:Is%20Not%20Valid
      (Variables|Default|SetStartLocation (Transformation|GetActorLocation Player))
      (Variables|Default|SetEndLocation (Transformation|GetActorLocation))))
  (Utilities|FlowControl|Sequence
    (:then_0
      (Utilities|FlowControl|Delay (+ (Components|Timeline|GetTimelineLength (Variables|Athena_Prop_SneakySnowmanV2|GetTimeline)) 0.1)))
    (:then_1
      (bind (_direction _rotationalpha) (Timeline|Timeline PlayFromStart)
        (:Update
          (Utilities|IsValid (Actor|GetAttachParentActor)
            (:Is%20Valid
              (Transformation|SetActorRelativeLocation Player (Math|Vector|Lerp(Vector) _startlocation (Math|Vector|MakeVector _x _y _returnvalue_2) _rotationalpha) false false))
            (:Is%20Not%20Valid
              (Transformation|SetActorLocation Player (Math|Vector|Lerp(Vector) _startlocation (Math|Vector|MakeVector _x _y _returnvalue_2) _rotationalpha) false false))))
        (:Finished
          _))))
  (CallFunction|DetachifAttachedActorforMoveIn true))

(event EventSpawnParticles
  (CallFunction|SpawnParticlesMulticast))

(event Custom|DetachIfAttachedActorForMoveIn
  (params (AllowLocal :type Boolean))
  (bind _actorlerpedformovein (Variables|Default|GetActorLerpedForMoveIn))
  ;[1 1 1] Only Detach if the actor we attached is still attached to the base used for the Move In Animation
  (Networking|SwitchHasAuthority
    (:Authority
      _)
    (:Remote
      (if AllowLocal
        _)))
  (Utilities|IsValid _actorlerpedformovein
    (:Is%20Valid
      (if (and (Utilities|IsValid (Variables|Default|GetAttachedBase) :!class /Script/Engine.KismetSystemLibrary) (== (Variables|Default|GetAttachedBase) (Actor|GetAttachParentActor _actorlerpedformovein)))
        (Transformation|DetachFromActor (Variables|Default|GetActorLerpedForMoveIn) "KeepWorld" "KeepWorld" "KeepWorld")))))

(event Custom|SpawnParticlesMulticast
  (bind _returnvalue (Transformation|GetActorLocation))
  (Utilities|FlowControl|DoOnce false
    (:Completed
      (Utilities|IsValid (Variables|FX|GetJumpIntoNiagara)
        (:Is%20Valid
          (Niagara|SpawnSystematLocation (Variables|FX|GetJumpIntoNiagara) _returnvalue))
        (:Is%20Not%20Valid
          ;[1 0.1719 0.1719] Remove after VFX revision
          (Effects|SpawnEmitteratLocation (Variables|FX|GetJumpinto) _returnvalue)))
      (Audio|PlaySoundatLocation (Variables|FX|GetJumpInSound) _returnvalue))))

(event EventOnTookSneakySnowman
  ;[1 1 1] Destroy self once player is in the snowman
  (Gameplay|SilentDie false)
  (Rendering|SetActorHiddenInGame true))

(event EventOnStartedJumpingIn
  (CallFunction|DisableCollision)
  (CallFunction|ApplyWaterinRadius))

(event Custom|DisableCollision
  (Collision|SetActorEnableCollision false)
  (Fortnite|SetCanbeDamaged self)
  (Utilities|Time|SetTimerbyEvent (delegate Custom|TimeOut) 4.0) ; Failsafe to make sure snowman is destroyed if something goes wrong
)

(event EventEndPlay
  (params (EndPlayReason :type EEndPlayReason%20Enum :default "Destroyed"))
  (bind _blockinteractiontag (Variables|Default|GetBlockInteractionTag))
  (|CallParentFunction EndPlayReason)
  (switch Utilities|FlowControl|Switch|SwitchonEEndPlayReason EndPlayReason
    (:Destroyed
      _)
    (:LevelTransition
      _)
    (:RemovedFromWorld
      _))
  (CallFunction|DetachifAttachedActorforMoveIn false)
  (bind _actorlerpedformovein (Variables|Default|GetActorLerpedForMoveIn)
    (:then
      (Ability|GameplayEffect|RemoveLooseGameplayTags _actorlerpedformovein _blockinteractiontag true)
      (Utilities|IsValid (Ability|GetFortAbilitySystemComponent _actorlerpedformovein)
        (:Is%20Valid
          (GameplayEffects|RemoveActiveEffectswithGrantedTags (Ability|GetFortAbilitySystemComponent _actorlerpedformovein) _blockinteractiontag))))
    (:else)))

(event Custom|TimeOut
  (Actor|DestroyActor))

(fn ApplyWaterInRadius
  (if (ScalableFloat|GetValueasBool (Variables|Default|GetShouldSpawnSnowball))
    (Networking|SwitchHasAuthority
      (:Authority
        (if (Curie|IsCurieActiveforActor self)
          ;[1 1 1] Create a list of Actors within the Radius
          (bind _outactors (Collision|SphereOverlapActors (Transformation|GetActorLocation) (* (Variables|Curie|GetWaterTraceRadius) (Math|Vector|VectorLength (Transformation|GetActorScale3D))) (Variables|Curie|GetWaterTraceObjectTypes) 0 (Utilities|Array|MakeArray self)))
          (for _array_element _outactors
            (Utilities|IsValid _array_element
              (:Is%20Valid
                (Curie|ExecuteWaterApplication (Variables|Actor|GetInstigator) self _array_element (Variables|Curie|GetWaterMagnitude) (Collision|MakeHitResult false true 0.0 0.0 (Transformation|GetActorLocation) (Transformation|GetActorLocation) "0,0,1" "0,0,1" 0 _array_element) (Variables|Curie|GetLandscapeExtinguishRadius))))))))))

(fn BlueprintCanInteract
  (params (InteractingPawn :type Fort%20Pawn%20Object%20Reference) (InteractionBeingAttempted :type EInteractionBeingAttempted%20Enum :default "FirstInteraction") (InteractionType :type TInteractionType%20Enum :default "IT_NoInteraction"))
  (Utilities|IsValid InteractingPawn
    (:Is%20Valid
      (bind _asfort_player_pawn (Utilities|Casting|CastToFortPlayerPawn InteractingPawn)
        (:then
          (bind _returnvalue (Fortnite|HeldObjects|GetHeldObjectAttachedtoPlayer _asfort_player_pawn))
          (Utilities|IsValid _returnvalue
            (:Is%20Valid
              (return false))
            (:Is%20Not%20Valid
              ;[1 1 1] Trace for Line Of Sight To Player. If the trace hits NOTHING, then you have line of sight. This is so the line of sight check works inside props that have no player collision. (Big Bush, Corn field)
              (bind (_outhit _returnvalue_2) (Collision|LineTraceByChannel (Transformation|GetActorLocation InteractingPawn) (+ (Transformation|GetActorLocation) "0,0.000000,20.000000") "TraceTypeQuery2" true (Utilities|Array|MakeArray InteractingPawn self) "None" false))
              (bind (_outhit_2 _returnvalue_3) (Collision|LineTraceByChannel (Transformation|GetActorLocation InteractingPawn) (+ (Transformation|GetActorLocation) "0,0.000000,200.000000") "TraceTypeQuery2" true (Utilities|Array|MakeArray InteractingPawn self) "None" false))
              (return (Math|Boolean|ANDBoolean (and (not _returnvalue_3) (not _returnvalue_2)) (and (or (not (GameplayTags|HasAnyMatchingGameplayTags InteractingPawn (GameplayTags|MakeGameplayTagContainerfromArray (Utilities|Array|MakeArray "(TagName=\"Gameplay.Status.HidingInProp\")" "(TagName=\"Gameplay.Status.HidingProp.GettingInProp\")" "(TagName=\"Granted.Athena.EnvItem.HidingProp.HidingInProp\")" "(TagName=\"Granted.Athena.EnvItem.HidingProp.GettingInProp\")")))) false) (not (GameplayTags|HasAnyMatchingGameplayTags InteractingPawn (GameplayTags|MakeGameplayTagContainerfromArray (Utilities|Array|MakeArray "(TagName=\"Gameplay.Action.Player.AthenaInVehicle\")" "(TagName=\"MovementMode.Falling\")" "(TagName=\"Gameplay.Status.HidingProp.BlockInteract\")" "(TagName=\"Gameplay.Action.Player.CarryingPlayer\")" "(TagName=\"Granted.HeldObject.Held\")" "(TagName=\"Granted.Athena.EnvItem.HidingProp.BlockInteract\")" "(TagName=\"Granted.Riding.Player.IsRiding\")" "(TagName=\"Gameplay.Status.DBNO\")"))))) (not (Variables|Default|GetDead)) (not (Utilities|IsValid (Variables|Default|GetActorLerpedForMoveIn) :!class /Script/Engine.KismetSystemLibrary)))))))
        (:CastFailed
          (return false))))
    (:Is%20Not%20Valid
      (return false))))

(fn BlueprintGetInteractionString
  (params (InteractingPawn :type Fort%20Pawn%20Object%20Reference) (InteractionBeingAttempted :type EInteractionBeingAttempted%20Enum :default "FirstInteraction"))
  (return "NSLOCTEXT(\"[C6201F7E4ABE3023B9397A9FE9193EB2]\", \"D7296C064CA17EB6EAAF23AB733FDAF7\", \"Hide\")"))

(fn BlueprintGetInteractionTime
  (params (InteractingPawn :type Fort%20Pawn%20Object%20Reference) (InteractionBeingAttempted :type EInteractionBeingAttempted%20Enum :default "FirstInteraction"))
  (return 0.0 true))

(fn BlueprintOnLocalInteract
  (params (InteractingPawn :type Fort%20Player%20Pawn%20Object%20Reference) (InteractionBeingAttempted :type EInteractionBeingAttempted%20Enum :default "FirstInteraction"))
  (bind _returnvalue (|CallParentFunction InteractingPawn InteractionBeingAttempted))
  ;[1 1 1] Immediately set values to avoid client sending invalid interaction events (fix for FORT-682801)
  (bind _output_get (Variables|Default|SetActorLerpedForMoveIn InteractingPawn))
  (Ability|GameplayEffect|AddLooseGameplayTags _output_get (Variables|Default|GetBlockInteractionTag))
  (return _returnvalue))

(fn GetMarkerDisplayData
  (return (Variables|marking|GetMarkerDisplay)))

(fn GetMaterialInt
  (return -1))

(fn GetNewSnowmanTransform
  (bind (_returnvalue_location _returnvalue_rotation _returnvalue_scale) (Math|Transform|BreakTransform (Transformation|GetActorTransform)))
  (return (+ _returnvalue_location "0, 0,25.000000") _returnvalue_rotation _returnvalue_scale))

(fn OnRep_bSimulationRunning)

(fn OnSpawned_NotOnPawn
  (Actor|DestroyActor))

(fn TraceStartAndEnd
  (return (Transformation|GetActorLocation) (- (Transformation|GetActorLocation) "0, 0,50.000000")))

(fn ConstructionScript
  (LOD|SetMaxDrawDistance (Variables|Debug|GetStaticMeshComponent) 0.0))
