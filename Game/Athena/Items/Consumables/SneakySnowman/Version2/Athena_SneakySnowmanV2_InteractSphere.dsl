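; Interaction sphere for the Sneaky Snowman V2 hiding consumable, offering an instant "Stop Hiding" prompt.
; Init links it to a snowman, which receives interaction events and gets StopHiding; the sphere then destroys itself, as it does when the snowman is destroyed.
; Interaction is blocked while popping out, tagged BlockInteract or in-vehicle, or when CheckifCanJumpOut fails.
; The focus point follows the hidden player's camera, offset forward and downward by FOV.
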(event Custom|Destroy
  (params (DestroyedActor :type Actor%20Object%20Reference))
  (Collision|SetActorEnableCollision false)
  (Actor|DestroyActor))

(event Interaction|EventBlueprintOnInteract
  (params (InteractingPawn :type Fort%20Pawn%20Object%20Reference) (InteractionBeingAttempted :type EInteractionBeingAttempted%20Enum :default "FirstInteraction"))
  (Class|AthenaPlayerSneakySnowmanV2|StopHiding (Variables|Default|GetAssociatedSnowman) true)
  (CallFunction|Destroy 0))

(event Interaction|EventBlueprintOnBeginInteract
  (params (InteractingPawn :type Fort%20Pawn%20Object%20Reference) (InteractionBeingAttempted :type EInteractionBeingAttempted%20Enum :default "FirstInteraction") (InteractionType :type TInteractionType%20Enum :default "IT_NoInteraction"))
  (Default|OnBeginInteractwithSphere (Variables|Default|GetAssociatedSnowman)))

(event Interaction|EventBlueprintOnInterruptInteract
  (Class|AthenaPlayerSneakySnowmanV2|OnInterruptInteractwithSphere (Variables|Default|GetAssociatedSnowman)))

(fn BlueprintCanInteract
  (params (InteractingPawn :type Fort%20Pawn%20Object%20Reference) (InteractionBeingAttempted :type EInteractionBeingAttempted%20Enum :default "FirstInteraction") (InteractionType :type TInteractionType%20Enum :default "IT_NoInteraction"))
  (bind _associatedsnowman (Variables|Default|GetAssociatedSnowman))
  (if (or (Variables|/Game/Athena/Items/Consumables/SneakySnowman/Version2/Athena_Player_SneakySnowmanV2.Athena_Player_SneakySnowmanV2_C|GetIsPoppingOut (Variables|Default|GetAssociatedSnowman)) (GameplayTags|HasAnyMatchingGameplayTags InteractingPawn (GameplayTags|MakeGameplayTagContainerfromArray (Utilities|Array|MakeArray "(TagName=\"Gameplay.Status.HidingProp.BlockInteract\")"))))
    (return false)
    (else
      (Utilities|IsValid _associatedsnowman
        (:Is%20Valid
          (bind _caninteract (Class|AthenaPlayerSneakySnowmanV2|CheckifCanJumpOut _associatedsnowman InteractingPawn))
          (return (and (not (GameplayTags|HasMatchingGameplayTag InteractingPawn "(TagName=\"Gameplay.Action.Player.AthenaInVehicle\")")) _caninteract)))
        (:Is%20Not%20Valid
          (return false))))))

(fn BlueprintGetInteractionString
  (params (InteractingPawn :type Fort%20Pawn%20Object%20Reference) (InteractionBeingAttempted :type EInteractionBeingAttempted%20Enum :default "FirstInteraction"))
  (return "NSLOCTEXT(\"[953E26574BFBF51A4110CC85DA05E4E5]\", \"A3D8CD9D48B8420A2EA800828152B8EF\", \"Stop Hiding\")"))

(fn BlueprintGetInteractionTime
  (params (InteractingPawn :type Fort%20Pawn%20Object%20Reference) (InteractionBeingAttempted :type EInteractionBeingAttempted%20Enum :default "FirstInteraction"))
  (return 0.0 true))

(fn DeleteOnResetMinigamePlayerIfAttached
  (return true))

(fn GetFocusedSocketLocation
  (Utilities|FlowControl|Sequence
    (:then_0
      (bind _associatedsnowman (Variables|Default|GetAssociatedSnowman)
        (:then
          (Utilities|IsValid (Fortnite|GetFortPlayerControllerfromActor (Variables|/Game/Athena/Items/Consumables/SneakySnowman/Version2/Athena_Player_SneakySnowmanV2.Athena_Player_SneakySnowmanV2_C|GetPlayerPawn _associatedsnowman))
            (:Is%20Valid
              (bind _output_get (LocalVariables|SetCameraManager (Variables|/Script/Engine.PlayerController|GetPlayerCameraManager (Fortnite|GetFortPlayerControllerfromActor (Variables|/Game/Athena/Items/Consumables/SneakySnowman/Version2/Athena_Player_SneakySnowmanV2.Athena_Player_SneakySnowmanV2_C|GetPlayerPawn _associatedsnowman)))))
              (Utilities|IsValid _output_get
                (:Is%20Valid
                  (return (Math|Vector|vector+vector (Camera|GetCameraLocation _output_get) (* (Math|Vector|GetForwardVector (Camera|GetCameraRotation _output_get)) (LocalVariables|GetForwardVectorMultipler)) (* (Math|Vector|GetUpVector (Camera|GetCameraRotation _output_get)) (* (Camera|GetFOVAngle _output_get) -1.0)))))))))
        (:else)))
    (:then_1
      ;[0 0 0 1 10] Default to what it would be if something went wrong, say AssociatedSnowman is pending kill
      (return (LocalVariables|GetForwardVectorMultipler) 0.0 0.0))))

(fn Init
  (params (AssociatedSnowman :type Athena%20Player%20Sneaky%20Snowman%20V2%20Object%20Reference))
  (bind _output_get (Variables|Default|SetAssociatedSnowman AssociatedSnowman))
  (Game|BindEventtoOnDestroyed _output_get (EventDispatchers|CreateEvent :!event Destroy)))

(fn OnRep_AssociatedSnowman)

(fn ConstructionScript)
