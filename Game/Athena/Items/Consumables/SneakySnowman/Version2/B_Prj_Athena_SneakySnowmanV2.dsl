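; Thrown Sneaky Snowman V2 projectile that puts the thrower into a snowman or spawns a snowman prop where it lands.
; BeginPlay removes depleted items from the instigator's inventory.
; On touch, upward-facing hits near an on-foot thrower apply a gameplay effect, while others can spawn the prop on authority after an overlap check, optionally attached to moving platforms.
; Explosion plays cosmetic FX; entering water plays FX and destroys the projectile.
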(event EventBeginPlay
  ;[1 1 1] EVENT -> BeginPlay - If no more charges then remove item
  (bind _asfort_player_pawn (Utilities|Casting|CastToFortPlayerPawn (Game|GetInstigator)))
  (if (not (GameplayTags|HasMatchingGameplayTag _asfort_player_pawn "(TagName=\"Athena.Item.Snowman\")"))
    (bind _asfort_player_controller (Utilities|Casting|CastToFortPlayerController (Pawn|GetController _asfort_player_pawn)))
    (Networking|SwitchHasAuthority
      (:Authority
        (Utilities|IsValid (Variables|SneakySnowman|GetSnowmanItemDef)
          (:Is%20Valid
            (Inventory|FindItemInstancesFromDefinition :self _asfort_player_controller :ItemDefinition (Variables|SneakySnowman|GetSnowmanItemDef) :ItemArray (Variables|Default|GetSnowmanItemInstances))
            (for _array_element (Variables|Default|GetSnowmanItemInstances)
              (if (== (Item|GetNuminStack _array_element) 0)
                (Inventory|RemoveFortItemFromPlayer _asfort_player_controller _array_element -1 true)))
            (Utilities|Array|Clear (Variables|Default|GetSnowmanItemInstances)))
          (:Is%20Not%20Valid
            (Inventory|RemoveItemFromPlayer :PlayerController _asfort_player_controller :ItemDefinition (Variables|SneakySnowman|GetSnowmanItemDef) :AmountToRemove -1 :bForceRemoval true :bSkipVerbRemoval false)))))))

(event EventOnExploded
  (params (HitActors :type Array%20of%20Actor%20Object%20References) (HitResults :type Array%20of%20Hit%20Result%20Structures))
  (bind _returnvalue (Transformation|GetActorLocation))
  (bind _snowmanspawnfx_niagara (Variables|FX|GetSnowmanSpawnFX_Niagara))
  ;[1 1 1] EVENT -> OnExploded - Audio and clean up
  (Networking|CanExecuteCosmeticEvents
    (:True
      (Utilities|IsValid _snowmanspawnfx_niagara
        (:Is%20Valid
          (Niagara|SpawnSystematLocation _snowmanspawnfx_niagara _returnvalue))
        (:Is%20Not%20Valid
          ;[1 0.1395 0.2392] Remove after VFX revision
          (Effects|SpawnEmitteratLocation (Variables|FX|GetSnowmanSpawnFX) _returnvalue)))
      (Camera|PlayWorldCameraShake "/Game/Athena/Items/Consumables/TowerGrenade/FX/B_Ground_CameraShake.B_Ground_CameraShake_C" _returnvalue 25.0 5e+01 2.0)
      (Audio|PlaySoundatLocation (Variables|Sounds|GetSoundOnExplo) _returnvalue)
      (Development|SetHiddeninGame (Variables|Default|GetSK_Sneaky_Snowman_Throw) true false)
      (Components|Activation|Deactivate (Variables|Sounds|GetAudio_InAir)))))

(event Custom|Apply%20Snowman%20To%20Self
  (bind _returnvalue (Game|GetInstigator))
  ;[1 0.1395 0.2392] Remove after VFX revision
  (Networking|SwitchHasAuthority
    (:Authority
      (bind _returnvalue_2 (GameplayTags|HasAnyMatchingGameplayTags(Message) _returnvalue (Variables|Default|GetTC_PreventApplicationTags)))
      (if (not _returnvalue_2)
        (Environment|ApplyGlobalEnvironmentGameplayEffecttoActor _returnvalue (Variables|SneakySnowman|GetGE_ApplyEnemySnowman)))
      (CallFunction|Explode))))

(event EventOnTouched
  (params (OtherActor :type Actor%20Object%20Reference) (OtherComp :type Primitive%20Component%20Object%20Reference) (HitResult :type Hit%20Result%20Structure%20%28by%20ref%29) (bIsOverlap :type Boolean))
  (bind (_returnvalue_roll _returnvalue_pitch _returnvalue_yaw) (Math|Rotator|BreakRotator (Transformation|GetActorRotation (Variables|Actor|GetInstigator))))
  (bind (_bblockinghit _binitialoverlap _time _distance _location _impactpoint _normal _impactnormal _physmat _hitactor) (Collision|BreakHitResult HitResult))
  (bind _instigator (Variables|Actor|GetInstigator))
  (bind _returnvalue (GameplayTags|HasMatchingGameplayTag(Message) OtherActor "(TagName=\"Gameplay.SneakySnowman.DontSpawn\")"))
  (if _returnvalue
    (CallFunction|Explode) ; If hit water/vehicle/flopper/Moving Platform, explode
    (else
      (Utilities|Casting|CastToFortPlayerPawn OtherActor
        (:CastFailed
          (Utilities|Casting|CastToB_HidingProp OtherActor
            (:CastFailed
              (Utilities|Casting|CastToFortAthenaVehicle OtherActor
                (:then
                  _)
                (:CastFailed
                  ;[1 1 1] EVENT -> On Stop do Hit FX and Spawn Snowman
                  (Utilities|Casting|CastToWaterBody OtherActor
                    (:then
                      _)
                    (:CastFailed
                      (Utilities|Casting|CastToBGA_Athena_FlopperSpawn_Parent OtherActor
                        (:then
                          _)
                        (:CastFailed
                          (Utilities|IsValid (Actor|GetComponentbyClass OtherActor "/Script/FortniteGame.FortMovingPlatformComponent")
                            (:Is%20Valid
                              (if (not (CallFunction|ShouldAttachtoMovingPlatform))
                                (CallFunction|Explode) ; If hit water/vehicle/flopper/Moving Platform, explode
                              ))
                            (:Is%20Not%20Valid
                              _))
                          (bind _asfort_player_pawn (Utilities|Casting|CastToFortPlayerPawn _instigator))
                          (if (not (Variables|SneakySnowman|GetHitPlayer))
                            (Variables|SneakySnowman|SetHitLocation _location)
                            (if (> (Math|Vector|DotProduct _normal "0, 0,1.000000") 0.3)
                              (if (and (<= (Math|Vector|VectorLength (- (Transformation|GetActorLocation _instigator) (Variables|SneakySnowman|GetHitLocation))) 1.5e+02) (not (Vehicle|IsinVehicle _asfort_player_pawn)))
                                (CallFunction|ApplySnowmantoSelf)
                                (else
                                  (Utilities|Casting|CastToAthena_Prop_SneakySnowmanV2 _hitactor
                                    (:then
                                      (Utilities|FlowControl|Delay 0.2) ; Safe check in case it got stuck in snowman's arm
                                      (if (<= (Math|Vector|VectorLength (Variables|/Script/Engine.MovementComponent|GetVelocity (Variables|Movement|GetProjectileMovementComponent))) 0.0)
                                        (CallFunction|Explode)))
                                    (:CastFailed
                                      (bind _asfort_player_pawn_2 (Utilities|Casting|CastToFortPlayerPawn _hitactor)
                                        (:then
                                          (if (CallFunction|ShouldApplySnowman? _asfort_player_pawn_2)
                                            _))
                                        (:CastFailed
                                          (Networking|SwitchHasAuthority
                                            (:Authority
                                              ;;[1 1 1] Bandaid for sh-03-20-underwhere
                                              (bind (_outactors _returnvalue_2) (Collision|SphereOverlapActors _location 1e+02 (Utilities|Array|MakeArray "ObjectTypeQuery5" "ObjectTypeQuery3"))) ; If we are very near to any of these Object Types, do not spawn the snowman.
                                              (if (not _returnvalue_2)
                                                (bind _returnvalue_3 (Building|SpawnBuildingActor (Variables|Default|GetPropClass) (Math|Transform|MakeTransform _location (Math|Rotator|MakeRotator 0.0 0.0 (+ _returnvalue_yaw -9e+01)) "1.000000,1.000000,1.000000") 0 (Variables|Actor|GetInstigator)))
                                                (bind _buildingactor (CallFunction|HandleAttachtoMovingPlatform _returnvalue_3 OtherActor))
                                                (Ability|SendGameplayEventtoActor (Game|GetInstigator) "(TagName=\"Abilities.Drops.AddToActorLimit.SneakySnowman\")" (Utilities|Struct|MakeGameplayEventData :Target _buildingactor :ContextHandle "()" :InstigatorTags "(GameplayTags=)" :TargetTags "(GameplayTags=)" :EventMagnitude 0.0 :TargetData "()"))))
                                            (:Remote
                                              _))))
                                      ;[1 1 1] Explode (Custom Event)
                                      (CallFunction|Explode))))))))))))))))))))

(event Custom|Explode
  ;[1 1 1] Explode (Custom Event)
  (Delivery|DoExplosion)
  ;;[1 1 1] Stop rotating movement
  (Components|Movement|SetUpdatedComponent (Variables|Default|GetRotatingMovement) 0))

(event OnEnterWaterEvent(WaterInteractionComponent)
  (params (WaterBody :type Fort%20Water%20Body%20Actor%20Object%20Reference) (WaterInteractionComponent :type Fort%20Water%20Interaction%20Component%20Object%20Reference) (bIsFirstBody :type Boolean))
  (bind _returnvalue (Transformation|GetActorLocation))
  (bind _enterwaterfx_niagara (Variables|FX|GetEnterWaterFX_Niagara))
  (Utilities|IsValid _enterwaterfx_niagara
    (:Is%20Valid
      (Niagara|SpawnSystematLocation _enterwaterfx_niagara _returnvalue))
    (:Is%20Not%20Valid
      ;[1 0.1042 0.1042] Remove after VFX revision
      (Effects|SpawnEmitteratLocation (Variables|FX|GetEnterWaterFX) _returnvalue)))
  (Rendering|SetActorHiddenInGame false)
  (Collision|SetActorEnableCollision false)
  (Utilities|FlowControl|Delay 0.5)
  (Actor|DestroyActor))

(fn HandleAttachToMovingPlatform
  (params (BuildingActor :type Building%20Actor%20Object%20Reference) (HitActor :type Actor%20Object%20Reference))
  (bind _returnvalue (Class|FortBaseMovingPlatformComponent|FindMovingPlatformComponentfromActor HitActor true))
  (if (CallFunction|ShouldAttachtoMovingPlatform)
    (Utilities|IsValid _returnvalue
      (:Is%20Valid
        (Transformation|AttachActorToActor BuildingActor (Components|GetOwner _returnvalue) "None" "KeepWorld" "KeepWorld" "KeepWorld" true)
        (Actor|SetLifeSpan BuildingActor (ScalableFloat|GetValueatLevel (Variables|Default|GetLifspanOnMovingPlatform)))
        (return BuildingActor)))
    (else
      (return BuildingActor))))

(fn ShouldBounce?
  (params (Hit%20Normal :type Vector))
  ;[1 1 1] Only bounce if we are on a steep enough slope
  (if (> (Math|Vector|DotProduct Hit%20Normal (Math|Vector|GetUpVector "0, 0, 0")) 0.65)
    (Variables|/Script/Engine.ProjectileMovementComponent|SetbShouldBounce false (Variables|Movement|GetProjectileMovementComponent))))

(fn ShouldApplySnowman?
  (params (PlayerPawn :type Object%20Reference))
  (bind _returnvalue (GameplayTags|HasAnyMatchingGameplayTags(Message) PlayerPawn (GameplayTags|MakeGameplayTagContainerfromArray (Utilities|Array|MakeArray "(TagName=\"Athena.Item.Snowman\")" "(TagName=\"Gameplay.Status.HidingProp.GettingInProp\")" "(TagName=\"Gameplay.Status.HidingInProp\")"))))
  (return (and (not _returnvalue) (Utilities|NotEqual(Object) PlayerPawn (Game|GetInstigator)))))

(fn ShouldAttachToMovingPlatform
  (return (ScalableFloat|GetValueasBool (Variables|Default|GetShouldSpawnOnMovingPlatforms))))

(fn UpdateQuest
  (params (FortPC :type Fort%20Player%20Controller%20Object%20Reference))
  ;[1 1 1] Quest hook (if this ever needs to change please let Dave.Ewing or Nicholas.Lacelle know)
  (Networking|SwitchHasAuthority
    (:Authority
      (Quest|SendComplexCustomStatEvent (Quests|GetQuestManager FortPC "Invalid") 0 "(GameplayTags=((TagName=\"Gameplay.Athena.Quests.UseSneakySnowman\")))" "(GameplayTags=)" 1))))

(fn ConstructionScript
  (Networking|CanExecuteCosmeticEvents
    (:True
      (Components|Movement|SetUpdatedComponent (Variables|Default|GetRotatingMovement) (Variables|Default|GetSK_Sneaky_Snowman_Throw))
      (Audio|SetCollisionSoundatIndex (Variables|Default|GetFortCollisionAudio) 0 (Variables|Sounds|GetCue_HitPlayer)))))
