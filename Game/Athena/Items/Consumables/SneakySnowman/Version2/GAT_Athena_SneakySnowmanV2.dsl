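; Gameplay ability for the Sneaky Snowman V2 consumable that strips the snowman effect from its owner when activated.
; On activation it calls the parent implementation, runs generic ability setup and stores the resulting player pawn.
; It then removes owner gameplay effects granting the Athena.Item.Snowman tag and forces a net update on that pawn so the change replicates promptly.
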(event Ability|EventActivateAbility
  (|CallParentFunction)
  (bind (_event_data _fort_player_pawn) (Setup|SetupGenericAbility false false)
    (:Output
      (Variables|Default|SetPlayerPawn _fort_player_pawn)
      (Ability|RemoveGameplayEffectFromOwnerWithGrantedTags "(GameplayTags=((TagName=\"Athena.Item.Snowman\")))")
      (Networking|ForceNetUpdate (Variables|Default|GetPlayerPawn)))))
